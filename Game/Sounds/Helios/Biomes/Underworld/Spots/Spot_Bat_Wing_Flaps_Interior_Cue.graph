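# Ambient one-shot cue of bat wing flaps, named as an Underworld interior spot, using bird attenuation and reusing Jungle Temple and Grassland waves.
# A quality switch gives its first two slots a layered mix of a random Jungle Temple bat sound and a random wing flap.
# Its last two slots play only one random flap from a smaller set.
# Both branches randomize pitch and volume through modulators.
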SoundCue /Game/Sounds/Helios/Biomes/Underworld/Spots/Spot_Bat_Wing_Flaps_Interior_Cue
root=QualityLevel_0
nodes [
 {id=Mixer_0, children [Random_0,Random_2], properties {inputVolume [1,1]}}
 {id=Modulator_0, children [Mixer_0], properties {pitchMax=1.1, pitchMin=0.9, volumeMax=1, volumeMin=0.5}}
 {id=Modulator_2, children [Random_3], properties {pitchMax=1.1, volumeMax=1.1, volumeMin=0.9}}
 {id=QualityLevel_0, children [Modulator_0,Modulator_0,Modulator_2,Modulator_2]}
 {
  id=Random_0
  children [WavePlayer_22, WavePlayer_23, WavePlayer_24, WavePlayer_25, WavePlayer_26, WavePlayer_27, WavePlayer_28]
  properties {weights [1,1,1,1,1,1,1]}
 }
 {id=Random_2, children [WavePlayer_8, WavePlayer_9, WavePlayer_10, WavePlayer_11, WavePlayer_12], properties {weights [1,1,1,1,1]}}
 {id=Random_3, children [WavePlayer_13, WavePlayer_14, WavePlayer_15, WavePlayer_16], properties {weights [1,1,1,1]}}
 {
  id=WavePlayer_10
  properties {soundWaveAssetPtr=/Game/Sounds/Artemis/Biomes/Grasslands/Grassland_Spots/Spot_Grassland_Night_Bat_Wing_Flap_05}
 }
 {
  id=WavePlayer_11
  properties {soundWaveAssetPtr=/Game/Sounds/Artemis/Biomes/Grasslands/Grassland_Spots/Spot_Grassland_Night_Bat_Wing_Flap_03}
 }
 {
  id=WavePlayer_12
  properties {soundWaveAssetPtr=/Game/Sounds/Artemis/Biomes/Grasslands/Grassland_Spots/Spot_Grassland_Night_Bat_Wing_Flap_01}
 }
 {
  id=WavePlayer_13
  properties {soundWaveAssetPtr=/Game/Sounds/Artemis/Biomes/Grasslands/Grassland_Spots/Spot_Grassland_Night_Bat_Wing_Flap_02}
 }
 {
  id=WavePlayer_14
  properties {soundWaveAssetPtr=/Game/Sounds/Artemis/Biomes/Grasslands/Grassland_Spots/Spot_Grassland_Night_Bat_Wing_Flap_04}
 }
 {
  id=WavePlayer_15
  properties {soundWaveAssetPtr=&Spot_Grassland_Night_Bat_Wing_Flap_05}
 }
 {
  id=WavePlayer_16
  properties {soundWaveAssetPtr=&Spot_Grassland_Night_Bat_Wing_Flap_03}
 }
 {id=WavePlayer_22, properties {soundWaveAssetPtr=/Game/Sounds/Artemis/POI/Jungle_Temple/Spots/AMB_Canary_Jungle_Temple_Int_SS_Bats_02_01}}
 {id=WavePlayer_23, properties {soundWaveAssetPtr=/Game/Sounds/Artemis/POI/Jungle_Temple/Spots/AMB_Canary_Jungle_Temple_Int_SS_Bats_02_02}}
 {id=WavePlayer_24, properties {soundWaveAssetPtr=/Game/Sounds/Artemis/POI/Jungle_Temple/Spots/AMB_Canary_Jungle_Temple_Int_SS_Bats_02_03}}
 {id=WavePlayer_25, properties {soundWaveAssetPtr=/Game/Sounds/Artemis/POI/Jungle_Temple/Spots/AMB_Canary_Jungle_Temple_Int_SS_Bats_02_04}}
 {id=WavePlayer_26, properties {soundWaveAssetPtr=/Game/Sounds/Artemis/POI/Jungle_Temple/Spots/AMB_Canary_Jungle_Temple_Int_SS_Bats_02_05}}
 {id=WavePlayer_27, properties {soundWaveAssetPtr=/Game/Sounds/Artemis/POI/Jungle_Temple/Spots/AMB_Canary_Jungle_Temple_Int_SS_Bats_02_06}}
 {id=WavePlayer_28, properties {soundWaveAssetPtr=/Game/Sounds/Artemis/POI/Jungle_Temple/Spots/AMB_Canary_Jungle_Temple_Int_SS_Bats_02_07}}
 {id=WavePlayer_8, properties {soundWaveAssetPtr=&Spot_Grassland_Night_Bat_Wing_Flap_02}}
 {id=WavePlayer_9, properties {soundWaveAssetPtr=&Spot_Grassland_Night_Bat_Wing_Flap_04}}
]
properties {
 attenuationSettings=&Fort_TileAmbient_OneShot_Bird_Att
 bEnableBusSends=false
 maxDistance=8005
 soundClassObject=&Fort_Ambient_OneShots
 volumeMultiplier=0.251
 preEffectBusSends [{}]
}
