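# Ambient one-shot spot cue for chain movement in the Underworld biome, repeating endlessly at random intervals with pitch and volume variation.
# A mixer layers a random slow chain rattle with a quieter random chain movement sound; every quality level uses this same mix.
# Each layer has its own custom attenuation: rattles add listener focus and a distance-scaled reverb send, while movement sounds fade out up close and peak farther away.
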SoundCue /Game/Sounds/Helios/Biomes/Underworld/Spots/Spot_Brimstone_Chain_Movement_Cue
root=Looping_0
nodes [
 {
  id=Attenuation_0
  children [Random_1]
  properties {
   bOverrideAttenuation=true
   attenuationOverrides {
    distanceAlgorithm=Custom
    falloffDistance=2000
    attenuationShapeExtents {x=0}
    customAttenuationCurve {editorCurveData {keys [
       <at 0 0>
       {arriveTangent=1.6667, interpMode=RCIM_Cubic, leaveTangent=1.6667, time=0.4, value=0.0018}
       {arriveTangent=-0.0017, interpMode=RCIM_Cubic, leaveTangent=-0.0017, time=0.6, value=1}
       <at 1.5 0>
      ]}}
   }
  }
 }
 {
  id=Attenuation_1
  children [Random_0]
  properties {
   bOverrideAttenuation=true
   attenuationOverrides {
    bEnableFocusInterpolation=true
    bEnableListenerFocus=true
    bEnableSubmixSends=true
    distanceAlgorithm=Custom
    falloffDistance=1500
    focusReleaseInterpSpeed=0.25
    nonFocusVolumeAttenuation=0.25
    attenuationShapeExtents {x=0}
    customAttenuationCurve {editorCurveData {keys [<at 0.0051 0.4771 RCIM_Cubic>, {arriveTangent=-0.4796, interpMode=RCIM_Cubic, leaveTangent=-0.4796, time=0.4, value=1}, <at 1 0>]}}
    submixSendSettings [{
      bEnableHPFCutoff=false
      bEnableLPFCutoff=false
      disableManualSendClamp=false
      hPFCutoff=20
      lPFCutoff=20000
      maxSendDistance=6000
      maxSendLevel=1
      minSendDistance=400
      minSendLevel=0
      sendLevel=1
      sendLevelControlMethod=Linear
      soundSubmix=&Fortnite_Reverb_Default
     }]
   }
  }
 }
 {id=Delay_0, children [Modulator_0], properties {delayMax=25, delayMin=8}}
 {id=Looping_0, children [Delay_0]}
 {id=Mixer_0, children [Attenuation_1, Attenuation_0], properties {inputVolume [0.3,0.1]}}
 {id=Modulator_0, children [QualityLevel_0], properties {pitchMin=0.8, volumeMin=0.7}}
 {id=QualityLevel_0, children [Mixer_0,Mixer_0,Mixer_0,Mixer_0]}
 {
  id=Random_0
  children [WavePlayer_0,WavePlayer_2,WavePlayer_3,WavePlayer_4,WavePlayer_5,WavePlayer_6,WavePlayer_7,WavePlayer_8]
  properties {bShouldExcludeFromBranchCulling=true, weights [1,1,1,1,1,1,1,1]}
 }
 {
  id=Random_1
  children [
   WavePlayer_1
   WavePlayer_9
   WavePlayer_10
   WavePlayer_11
   WavePlayer_12
   WavePlayer_13
   WavePlayer_14
   WavePlayer_15
   WavePlayer_16
   WavePlayer_17
   WavePlayer_18
   WavePlayer_19
   WavePlayer_20
  ]
  properties {weights [1,1,1,1,1,1,1,1,1,1,1,1,1]}
 }
 {id=WavePlayer_0, properties {soundWaveAssetPtr=/Game/Sounds/Fort_Starlight/S1/SoundWaves/SpotFX_ChainRattle_Slow_01}}
 {id=WavePlayer_1, properties {soundWaveAssetPtr=&/Spot_Chain_Movement_01}}
 {id=WavePlayer_10, properties {soundWaveAssetPtr=&/Spot_Chain_Movement_03}}
 {id=WavePlayer_11, properties {soundWaveAssetPtr=&/Spot_Chain_Movement_04}}
 {id=WavePlayer_12, properties {soundWaveAssetPtr=&/Spot_Chain_Movement_05}}
 {id=WavePlayer_13, properties {soundWaveAssetPtr=&/Spot_Chain_Movement_06}}
 {id=WavePlayer_14, properties {soundWaveAssetPtr=&/Spot_Chain_Movement_07}}
 {id=WavePlayer_15, properties {soundWaveAssetPtr=&/Spot_Chain_Movement_08}}
 {id=WavePlayer_16, properties {soundWaveAssetPtr=&/Spot_Chain_Movement_09}}
 {id=WavePlayer_17, properties {soundWaveAssetPtr=&/Spot_Chain_Movement_10}}
 {id=WavePlayer_18, properties {soundWaveAssetPtr=&/Spot_Chain_Movement_11}}
 {id=WavePlayer_19, properties {soundWaveAssetPtr=&/Spot_Chain_Movement_12}}
 {id=WavePlayer_2, properties {soundWaveAssetPtr=/Game/Sounds/Fort_Starlight/S1/SoundWaves/SpotFX_ChainRattle_Slow_02}}
 {id=WavePlayer_20, properties {soundWaveAssetPtr=&/Spot_Chain_Movement_13}}
 {id=WavePlayer_3, properties {soundWaveAssetPtr=/Game/Sounds/Fort_Starlight/S1/SoundWaves/SpotFX_ChainRattle_Slow_03}}
 {id=WavePlayer_4, properties {soundWaveAssetPtr=/Game/Sounds/Fort_Starlight/S1/SoundWaves/SpotFX_ChainRattle_Slow_04}}
 {id=WavePlayer_5, properties {soundWaveAssetPtr=/Game/Sounds/Fort_Starlight/S1/SoundWaves/SpotFX_ChainRattle_Slow_05}}
 {id=WavePlayer_6, properties {soundWaveAssetPtr=/Game/Sounds/Fort_Starlight/S1/SoundWaves/SpotFX_ChainRattle_Slow_06}}
 {id=WavePlayer_7, properties {soundWaveAssetPtr=/Game/Sounds/Fort_Starlight/S1/SoundWaves/SpotFX_ChainRattle_Slow_07}}
 {id=WavePlayer_8, properties {soundWaveAssetPtr=/Game/Sounds/Fort_Starlight/S1/SoundWaves/SpotFX_ChainRattle_Slow_08}}
 {id=WavePlayer_9, properties {soundWaveAssetPtr=&/Spot_Chain_Movement_02}}
]
properties {
 bOverrideAttenuation=true
 maxDistance=3000
 soundClassObject=&Fort_Ambient_OneShots
 volumeMultiplier=0.251
 attenuationOverrides {
  bAttenuateWithLPF=true
  bEnableOcclusion=true
  distanceAlgorithm=Custom
  falloffDistance=2500
  lPFFrequencyAtMax=1000
  lPFRadiusMax=3000
  lPFRadiusMin=1000
  occlusionInterpolationTime=3
  occlusionLowPassFilterFrequency=1000
  occlusionTraceChannel=ECC_GameTraceChannel11
  occlusionVolumeAttenuation=0.5
  stereoSpread=50
  attenuationShapeExtents {x=0}
  customAttenuationCurve {editorCurveData {keys [
     <at 0 0>
     {arriveTangent=2.5, interpMode=RCIM_Cubic, leaveTangent=2.5, time=0.25, value=0.0018}
     <at 0.4 1>
     {arriveTangent=-0.0071, interpMode=RCIM_Cubic, leaveTangent=-0.0071, tangentMode=RCTM_User, time=0.6, value=1}
     <at 1 0>
    ]}}
 }
}
