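# Sound cue for quiet, distant chain-rattle ambient one-shot spots in the Underworld biome, named for Brimstone.
# An equal-weight random pick of thirteen chain-movement waves feeds all quality levels, then a modulator varies pitch and volume.
# Overridden attenuation uses a custom curve that is silent at the source, peaks at mid-range and fades out at max distance, with distance low-pass filtering and occlusion.
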SoundCue /Game/Sounds/Helios/Biomes/Underworld/Spots/Spot_Brimstone_Chain_Rattles_Distant_Cue
root=Modulator_0
nodes [
 {id=Modulator_0, children [QualityLevel_0], properties {pitchMin=0.8, volumeMin=0.8}}
 {id=QualityLevel_0, children [Random_1,Random_1,Random_1,Random_1]}
 {
  id=Random_1
  children [
   WavePlayer_1
   WavePlayer_9
   WavePlayer_10
   WavePlayer_11
   WavePlayer_12
   WavePlayer_13
   WavePlayer_14
   WavePlayer_15
   WavePlayer_16
   WavePlayer_17
   WavePlayer_18
   WavePlayer_19
   WavePlayer_20
  ]
  properties {weights [1,1,1,1,1,1,1,1,1,1,1,1,1]}
 }
 {id=WavePlayer_1, properties {soundWaveAssetPtr=&/Spot_Chain_Movement_01}}
 {id=WavePlayer_10, properties {soundWaveAssetPtr=&/Spot_Chain_Movement_03}}
 {id=WavePlayer_11, properties {soundWaveAssetPtr=&/Spot_Chain_Movement_04}}
 {id=WavePlayer_12, properties {soundWaveAssetPtr=&/Spot_Chain_Movement_05}}
 {id=WavePlayer_13, properties {soundWaveAssetPtr=&/Spot_Chain_Movement_06}}
 {id=WavePlayer_14, properties {soundWaveAssetPtr=&/Spot_Chain_Movement_07}}
 {id=WavePlayer_15, properties {soundWaveAssetPtr=&/Spot_Chain_Movement_08}}
 {id=WavePlayer_16, properties {soundWaveAssetPtr=&/Spot_Chain_Movement_09}}
 {id=WavePlayer_17, properties {soundWaveAssetPtr=&/Spot_Chain_Movement_10}}
 {id=WavePlayer_18, properties {soundWaveAssetPtr=&/Spot_Chain_Movement_11}}
 {id=WavePlayer_19, properties {soundWaveAssetPtr=&/Spot_Chain_Movement_12}}
 {id=WavePlayer_20, properties {soundWaveAssetPtr=&/Spot_Chain_Movement_13}}
 {id=WavePlayer_9, properties {soundWaveAssetPtr=&/Spot_Chain_Movement_02}}
]
properties {
 bOverrideAttenuation=true
 maxDistance=2500
 soundClassObject=&Fort_Ambient_OneShots
 volumeMultiplier=0.05
 attenuationOverrides {
  bAttenuateWithLPF=true
  bEnableOcclusion=true
  distanceAlgorithm=Custom
  falloffDistance=2500
  lPFFrequencyAtMax=1000
  lPFRadiusMax=3000
  lPFRadiusMin=1000
  occlusionInterpolationTime=3
  occlusionLowPassFilterFrequency=1000
  occlusionTraceChannel=ECC_GameTraceChannel11
  occlusionVolumeAttenuation=0.5
  stereoSpread=100
  attenuationShapeExtents {x=0}
  customAttenuationCurve {editorCurveData {keys [
     <at 0 0>
     {arriveTangent=2.5, interpMode=RCIM_Cubic, leaveTangent=2.5, time=0.25, value=0.0018}
     <at 0.4 1>
     {arriveTangent=-0.0071, interpMode=RCIM_Cubic, leaveTangent=-0.0071, tangentMode=RCTM_User, time=0.6, value=1}
     <at 1 0>
    ]}}
 }
}
