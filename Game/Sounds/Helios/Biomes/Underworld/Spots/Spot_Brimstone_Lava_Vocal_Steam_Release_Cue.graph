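# Sound cue for a brimstone lava vocal steam-release spot one-shot in the Helios Underworld biome ambience.
# A Random node picks one of seven equally weighted steam-release wave variants, then a Modulator varies its pitch for extra variety.
# Playback is quieted, routed to the ambient one-shots sound class with bus and submix sends disabled, and uses a less-positional log tile-ambient attenuation asset.
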SoundCue /Game/Sounds/Helios/Biomes/Underworld/Spots/Spot_Brimstone_Lava_Vocal_Steam_Release_Cue
root=Modulator_0
nodes [
 {id=Modulator_0, children [Random_0], properties {pitchMin=0.9}}
 {
  id=Random_0
  children [WavePlayer_0,WavePlayer_1,WavePlayer_2,WavePlayer_3,WavePlayer_4,WavePlayer_5,WavePlayer_6]
  properties {weights [1,1,1,1,1,1,1]}
 }
 {id=WavePlayer_0, properties {soundWaveAssetPtr=&/Spot_Brimstone_Lava_Vocal_Steam_Release_01}}
 {id=WavePlayer_1, properties {soundWaveAssetPtr=&/Spot_Brimstone_Lava_Vocal_Steam_Release_02}}
 {id=WavePlayer_2, properties {soundWaveAssetPtr=&/Spot_Brimstone_Lava_Vocal_Steam_Release_03}}
 {id=WavePlayer_3, properties {soundWaveAssetPtr=&/Spot_Brimstone_Lava_Vocal_Steam_Release_04}}
 {id=WavePlayer_4, properties {soundWaveAssetPtr=&/Spot_Brimstone_Lava_Vocal_Steam_Release_05}}
 {id=WavePlayer_5, properties {soundWaveAssetPtr=&/Spot_Brimstone_Lava_Vocal_Steam_Release_06}}
 {id=WavePlayer_6, properties {soundWaveAssetPtr=&/Spot_Brimstone_Lava_Vocal_Steam_Release_07}}
]
properties {
 attenuationSettings=/Game/Sounds/Attenuation/Fort_TileAmbient_OneShot_LessPositional_Log_Att
 bEnableBusSends=false
 bEnableSubmixSends=false
 maxDistance=6200
 soundClassObject=&Fort_Ambient_OneShots
 volumeMultiplier=0.301
 attenuationOverrides {falloffDistance=6000, stereoSpread=10000, attenuationShapeExtents {x=2000}}
}
