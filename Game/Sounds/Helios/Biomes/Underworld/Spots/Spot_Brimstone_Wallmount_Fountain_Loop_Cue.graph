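# Looping ambient spot sound for a wall-mounted brimstone fountain in the Underworld biome, built from flowing-lava loops.
# A random pick of four looping waves gets pitch modulation, then passes through three nested endlessly looping envelopes that drift pitch and layer irregular and slow volume swells.
# All quality levels share this chain.
# Overrides give short-range capsule attenuation with occlusion and listener focus, plus a two-voice limit; output also feeds an exterior ambience bus.
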SoundCue /Game/Sounds/Helios/Biomes/Underworld/Spots/Spot_Brimstone_Wallmount_Fountain_Loop_Cue
root=QualityLevel_0
nodes [
 {
  id=Enveloper_1
  children [Modulator_0]
  properties {
   bLoop=true
   bLoopIndefinitely=true
   loopEnd=60
   pitchCurve {editorCurveData {keys [
      {interpMode=RCIM_Cubic, time=0}
      <at 15 0.85 RCIM_Cubic>
      {arriveTangent=0.0091, interpMode=RCIM_Cubic, leaveTangent=0.0091, tangentMode=RCTM_User, time=30}
      {interpMode=RCIM_Cubic, time=45}
      {time=60}
     ]}}
   volumeCurve {editorCurveData {keys [<at 0 1>]}}
  }
 }
 {
  id=Enveloper_3
  children [Enveloper_1]
  properties {
   bLoop=true
   bLoopIndefinitely=true
   loopEnd=60
   pitchCurve {editorCurveData {keys [{time=0}]}}
   volumeCurve {editorCurveData {keys [
      {arriveTangent=-0.01, interpMode=RCIM_Cubic, time=0, value=1}
      {arriveTangent=-0.0035, interpMode=RCIM_Cubic, leaveTangent=-0.0035, tangentMode=RCTM_User, time=4.5428, value=0.8801}
      {arriveTangent=0.0016, interpMode=RCIM_Cubic, leaveTangent=0.0016, tangentMode=RCTM_User, time=7.4135, value=0.9776}
      {arriveTangent=0.0008, interpMode=RCIM_Cubic, leaveTangent=0.0008, tangentMode=RCTM_User, time=11.1077, value=0.8936}
      {arriveTangent=0.0016, interpMode=RCIM_Cubic, leaveTangent=0.0016, tangentMode=RCTM_User, time=12.9775, value=0.8792}
      {arriveTangent=0.0009, interpMode=RCIM_Cubic, leaveTangent=0.0009, time=14.965, value=0.9463}
      {arriveTangent=-0.0011, interpMode=RCIM_Cubic, leaveTangent=-0.0011, tangentMode=RCTM_User, time=18.5868, value=0.8842}
      {arriveTangent=0.0002, interpMode=RCIM_Cubic, leaveTangent=0.0002, time=23.9559, value=0.9962}
      {arriveTangent=0.0016, interpMode=RCIM_Cubic, leaveTangent=0.0016, tangentMode=RCTM_User, time=28.034, value=0.8864}
      {arriveTangent=-0.0003, interpMode=RCIM_Cubic, leaveTangent=-0.0003, time=31.2025, value=0.961}
      {arriveTangent=0.0009, interpMode=RCIM_Cubic, leaveTangent=0.0009, tangentMode=RCTM_User, time=34.9227, value=0.8844}
      {arriveTangent=-0.0008, interpMode=RCIM_Cubic, leaveTangent=-0.0008, time=37.93, value=0.9962}
      <at 42.0081 0.879>
      {arriveTangent=0.0024, interpMode=RCIM_Cubic, leaveTangent=0.0024, tangentMode=RCTM_User, time=44.5668, value=0.8842}
      {arriveTangent=-0.0006, interpMode=RCIM_Cubic, leaveTangent=-0.0006, time=47.958, value=0.9776}
      {arriveTangent=-0.0056, interpMode=RCIM_Cubic, leaveTangent=-0.0056, tangentMode=RCTM_User, time=53.1343, value=0.8792}
      <at 55.3163 0.9711>
      {arriveTangent=-0.0003, interpMode=RCIM_Cubic, leaveTangent=-0.0003, tangentMode=RCTM_User, time=57.7536, value=0.8789}
      <at 60 1 RCIM_Cubic>
     ]}}
  }
 }
 {
  id=Enveloper_4
  children [Enveloper_3]
  properties {
   bLoop=true
   bLoopIndefinitely=true
   loopEnd=20
   pitchCurve {editorCurveData {keys [{time=0}]}}
   volumeCurve {editorCurveData {keys [
      {arriveTangent=-0.01, interpMode=RCIM_Cubic, time=0, value=1}
      <at 10 1.2 RCIM_Cubic>
      {interpMode=RCIM_Cubic, leaveTangent=0.01, time=20, value=1}
     ]}}
  }
 }
 {id=Modulator_0, children [Random_3], properties {pitchMin=0.8}}
 {id=QualityLevel_0, children [Enveloper_4,Enveloper_4,Enveloper_4,Enveloper_4]}
 {id=Random_3, children [WavePlayer_13, WavePlayer_14, WavePlayer_15, WavePlayer_16], properties {weights [1,1,1,1]}}
 {
  id=WavePlayer_13
  properties {bLooping=true, soundWaveAssetPtr=&/Spot_Brimstone_Lava_Flowing_Loop_01}
 }
 {
  id=WavePlayer_14
  properties {bLooping=true, soundWaveAssetPtr=&/Spot_Brimstone_Lava_Flowing_Loop_02}
 }
 {
  id=WavePlayer_15
  properties {bLooping=true, soundWaveAssetPtr=&/Spot_Brimstone_Lava_Flowing_Loop_03}
 }
 {
  id=WavePlayer_16
  properties {bLooping=true, soundWaveAssetPtr=&/Spot_Brimstone_Lava_Flowing_Loop_04}
 }
]
properties {
 attenuationSettings=/Game/Sounds/Helios/Environment_Actors/SewerPipes/Spot_UndergroundPipe_Loop_Capsule_Att
 bEnableSubmixSends=false
 bOverrideAttenuation=true
 bOverrideConcurrency=true
 maxDistance=600
 pitchMultiplier=1.2
 soundClassObject=&Fort_Ambient_Athena
 volumeMultiplier=0.501
 attenuationOverrides {
  attenuationShape=Capsule
  bEnableListenerFocus=true
  bEnableOcclusion=true
  falloffDistance=300
  focusAzimuth=25
  nonFocusAzimuth=45
  nonFocusVolumeAttenuation=0.75
  occlusionInterpolationTime=0.2
  occlusionLowPassFilterFrequency=1000
  occlusionTraceChannel=ECC_GameTraceChannel11
  occlusionVolumeAttenuation=0.5
  stereoSpread=100
  attenuationShapeExtents=<300 100>
 }
 busSends [{audioBus=null, minSendLevel=0.6, sendLevel=1, soundSourceBus=&AmbientAudio_Exterior_SourceBus}]
 concurrencyOverrides {maxCount=2}
 concurrencySet [&Weapon_Drop_StopOldest]
}
