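# Sound cue for distant one-shot cloth flap ambience spots in the Helios Underworld biome.
# It randomly plays either a large pirate-cove sail flap or a pitched-down pair of flag cloth flaps, the second randomly delayed, both branches with volume and pitch modulation.
# The first two quality slots share this chain; the other two are empty.
# Capsule attenuation adds occlusion and manual reverb, and output also feeds the exterior ambient bus.
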SoundCue /Game/Sounds/Helios/Biomes/Underworld/Spots/Spot_Cloth_Flaps_Oneshots_Distant_Cue
root=QualityLevel_1
nodes [
 {id=Delay_3, children [Random_5], properties {delayMax=2, delayMin=0.5}}
 {id=Mixer_1, children [Random_2], properties {inputVolume [0.8]}}
 {id=Mixer_2, children [Random_3,Delay_3], properties {inputVolume [1,0.8]}}
 {id=Modulator_1, children [Mixer_1], properties {pitchMax=1.1, volumeMax=1.2, volumeMin=0.9}}
 {id=Modulator_2, children [Mixer_2], properties {pitchMax=0.9, pitchMin=0.8, volumeMax=1.2, volumeMin=0.9}}
 {id=QualityLevel_1, children [Random_4,Random_4,null,null]}
 {
  id=Random_2
  children [
   WavePlayer_1
   WavePlayer_2
   WavePlayer_3
   WavePlayer_4
   WavePlayer_5
   WavePlayer_6
   WavePlayer_7
   WavePlayer_8
   WavePlayer_9
   WavePlayer_10
   WavePlayer_11
  ]
  properties {weights [1,1,1,1,1,1,1,1,1,1,1]}
 }
 {id=Random_3, children [WavePlayer_15, WavePlayer_16, WavePlayer_17], properties {weights [1,1,1]}}
 {id=Random_4, children [Modulator_1,Modulator_2], properties {weights [1,1]}}
 {id=Random_5, children [WavePlayer_18, WavePlayer_19, WavePlayer_20], properties {weights [1,1,1]}}
 {id=WavePlayer_1, properties {soundWaveAssetPtr=/Game/Sounds/Artemis/POI/Chrome_BugleBay/Spot_PirateCove_SailFlap_Large_01}}
 {id=WavePlayer_10, properties {soundWaveAssetPtr=/Game/Sounds/Artemis/POI/Chrome_BugleBay/Spot_PirateCove_SailFlap_Large_10}}
 {id=WavePlayer_11, properties {soundWaveAssetPtr=/Game/Sounds/Artemis/POI/Chrome_BugleBay/Spot_PirateCove_SailFlap_Large_11}}
 {id=WavePlayer_15, properties {soundWaveAssetPtr=/Game/Sounds/Helios/Environment_Actors/FlagBanner/Spot_Cloth_Flap_Oneshot_04}}
 {id=WavePlayer_16, properties {soundWaveAssetPtr=/Game/Sounds/Helios/Environment_Actors/FlagBanner/Spot_Cloth_Flap_Oneshot_05}}
 {id=WavePlayer_17, properties {soundWaveAssetPtr=/Game/Sounds/Helios/Environment_Actors/FlagBanner/Spot_Cloth_Flap_Oneshot_06}}
 {id=WavePlayer_18, properties {soundWaveAssetPtr=/Game/Sounds/Helios/Environment_Actors/FlagBanner/Spot_Cloth_Flap_Oneshot_01}}
 {id=WavePlayer_19, properties {soundWaveAssetPtr=/Game/Sounds/Helios/Environment_Actors/FlagBanner/Spot_Cloth_Flap_Oneshot_02}}
 {id=WavePlayer_2, properties {soundWaveAssetPtr=/Game/Sounds/Artemis/POI/Chrome_BugleBay/Spot_PirateCove_SailFlap_Large_02}}
 {id=WavePlayer_20, properties {soundWaveAssetPtr=/Game/Sounds/Helios/Environment_Actors/FlagBanner/Spot_Cloth_Flap_Oneshot_03}}
 {id=WavePlayer_3, properties {soundWaveAssetPtr=/Game/Sounds/Artemis/POI/Chrome_BugleBay/Spot_PirateCove_SailFlap_Large_03}}
 {id=WavePlayer_4, properties {soundWaveAssetPtr=/Game/Sounds/Artemis/POI/Chrome_BugleBay/Spot_PirateCove_SailFlap_Large_04}}
 {id=WavePlayer_5, properties {soundWaveAssetPtr=/Game/Sounds/Artemis/POI/Chrome_BugleBay/Spot_PirateCove_SailFlap_Large_05}}
 {id=WavePlayer_6, properties {soundWaveAssetPtr=/Game/Sounds/Artemis/POI/Chrome_BugleBay/Spot_PirateCove_SailFlap_Large_06}}
 {id=WavePlayer_7, properties {soundWaveAssetPtr=/Game/Sounds/Artemis/POI/Chrome_BugleBay/Spot_PirateCove_SailFlap_Large_07}}
 {id=WavePlayer_8, properties {soundWaveAssetPtr=/Game/Sounds/Artemis/POI/Chrome_BugleBay/Spot_PirateCove_SailFlap_Large_08}}
 {id=WavePlayer_9, properties {soundWaveAssetPtr=/Game/Sounds/Artemis/POI/Chrome_BugleBay/Spot_PirateCove_SailFlap_Large_09}}
]
properties {
 attenuationSettings=/Game/Sounds/Attenuation/Fort_TileAmbient_OneShot_Distant_Creature_Att
 maxDistance=8500
 soundClassObject=&Fort_Ambient_OneShots
 volumeMultiplier=0.251
 attenuationOverrides {
  attenuationShape=Capsule
  bEnableFocusInterpolation=true
  bEnableOcclusion=true
  falloffDistance=2500
  focusAttackInterpSpeed=3
  focusReleaseInterpSpeed=3
  manualReverbSendLevel=10
  nonFocusVolumeAttenuation=0.5
  occlusionLowPassFilterFrequency=3000
  occlusionTraceChannel=ECC_GameTraceChannel11
  reverbDistanceMax=2000
  reverbSendMethod=Manual
  reverbWetLevelMax=5
  attenuationShapeExtents {x=0}
 }
 busSends [{audioBus=null, minSendLevel=1, sendLevel=1, soundSourceBus=&AmbientAudio_Exterior_SourceBus}]
}
