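# Sound cue for quiet one-shot ghost movement swell ambience spots in the Helios Underworld biome.
# A quality-level switch feeds the first two levels from a seven-wave random pool and the last two from a four-wave subset excluded from branch culling.
# Each pool passes through slight pitch and volume modulation.
# Playback uses wide tile-ambient attenuation, a reverb submix send and a two-voice concurrency override.
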SoundCue /Game/Sounds/Helios/Biomes/Underworld/Spots/Spot_Ghost_Movement_Swells_02_Cue
root=QualityLevel_0
nodes [
 {id=Modulator_0, children [Random_2], properties {pitchMax=1.1, pitchMin=0.9, volumeMax=1.1, volumeMin=0.9}}
 {id=Modulator_1, children [Random_3], properties {pitchMax=1.1, pitchMin=0.9, volumeMax=1.1, volumeMin=0.9}}
 {id=QualityLevel_0, children [Modulator_0,Modulator_0,Modulator_1,Modulator_1]}
 {
  id=Random_2
  children [WavePlayer_26, WavePlayer_27, WavePlayer_28, WavePlayer_30, WavePlayer_32, WavePlayer_33, WavePlayer_34]
  properties {weights [1,1,1,1,1,1,1]}
 }
 {
  id=Random_3
  children [WavePlayer_0, WavePlayer_36, WavePlayer_38, WavePlayer_1]
  properties {bShouldExcludeFromBranchCulling=true, weights [1,1,1,1]}
 }
 {id=WavePlayer_0, properties {soundWaveAssetPtr=&/Spot_Ghost_Movement_Swells_08}}
 {id=WavePlayer_1, properties {soundWaveAssetPtr=&/Spot_Ghost_Movement_Swells_09}}
 {id=WavePlayer_26, properties {soundWaveAssetPtr=&/Spot_Ghost_Movement_Swells_01}}
 {id=WavePlayer_27, properties {soundWaveAssetPtr=&/Spot_Ghost_Movement_Swells_02}}
 {id=WavePlayer_28, properties {soundWaveAssetPtr=&/Spot_Ghost_Movement_Swells_03}}
 {id=WavePlayer_30, properties {soundWaveAssetPtr=&/Spot_Ghost_Movement_Swells_05}}
 {id=WavePlayer_32, properties {soundWaveAssetPtr=&/Spot_Ghost_Movement_Swells_07}}
 {id=WavePlayer_33, properties {soundWaveAssetPtr=&Spot_Ghost_Movement_Swells_08}}
 {id=WavePlayer_34, properties {soundWaveAssetPtr=&Spot_Ghost_Movement_Swells_09}}
 {id=WavePlayer_36, properties {soundWaveAssetPtr=&Spot_Ghost_Movement_Swells_05}}
 {id=WavePlayer_38, properties {soundWaveAssetPtr=&Spot_Ghost_Movement_Swells_02}}
]
properties {
 attenuationSettings=/Game/Sounds/Attenuation/Fort_TileAmbient_OneShot_Wider_Att
 bOverrideConcurrency=true
 maxDistance=8000
 soundClassObject=&Fort_Ambient_OneShots
 volumeMultiplier=0.251
 concurrencyOverrides {maxCount=2, voiceStealReleaseTime=2}
 soundSubmixSends [{minSendLevel=1, soundSubmix=&Fortnite_Reverb_Default}]
}
