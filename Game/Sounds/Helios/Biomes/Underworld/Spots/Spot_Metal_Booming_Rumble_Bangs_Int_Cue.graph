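# Sound cue for interior metal booming, rumbling bangs, a one-shot ambient spot effect in the Helios Underworld biome.
# A Random node picks among nine equally weighted wave variations.
# It plays quietly through the ambient one-shot sound class with a tile ambient one-shot attenuation.
# A concurrency override allows only one instance at a time and blocks new plays rather than cutting the active one.
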SoundCue /Game/Sounds/Helios/Biomes/Underworld/Spots/Spot_Metal_Booming_Rumble_Bangs_Int_Cue
root=Random_1
nodes [
 {
  id=Random_1
  children [WavePlayer_0, WavePlayer_1, WavePlayer_2, WavePlayer_3, WavePlayer_4, WavePlayer_5, WavePlayer_6, WavePlayer_14, WavePlayer_15]
  properties {weights [1,1,1,1,1,1,1,1,1]}
 }
 {id=WavePlayer_0, properties {soundWaveAssetPtr=&/Spot_Metal_Booming_Rumble_Bangs_Int_01}}
 {id=WavePlayer_1, properties {soundWaveAssetPtr=&/Spot_Metal_Booming_Rumble_Bangs_Int_02}}
 {id=WavePlayer_14, properties {soundWaveAssetPtr=&/Spot_Metal_Booming_Rumble_Bangs_Int_08}}
 {id=WavePlayer_15, properties {soundWaveAssetPtr=&/Spot_Metal_Booming_Rumble_Bangs_Int_09}}
 {id=WavePlayer_2, properties {soundWaveAssetPtr=&/Spot_Metal_Booming_Rumble_Bangs_Int_03}}
 {id=WavePlayer_3, properties {soundWaveAssetPtr=&/Spot_Metal_Booming_Rumble_Bangs_Int_04}}
 {id=WavePlayer_4, properties {soundWaveAssetPtr=&/Spot_Metal_Booming_Rumble_Bangs_Int_05}}
 {id=WavePlayer_5, properties {soundWaveAssetPtr=&/Spot_Metal_Booming_Rumble_Bangs_Int_06}}
 {id=WavePlayer_6, properties {soundWaveAssetPtr=&/Spot_Metal_Booming_Rumble_Bangs_Int_07}}
]
properties {
 attenuationSettings=/Game/Sounds/Attenuation/Fort_TileAmbient_OneShot_Att
 bEnableBusSends=false
 bEnableSubmixSends=false
 bOverrideConcurrency=true
 maxDistance=8000
 soundClassObject=&Fort_Ambient_OneShots
 volumeMultiplier=0.301
 concurrencyOverrides {maxCount=1, resolutionRule=PreventNew, voiceStealReleaseTime=5}
}
