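# Sound cue for Underworld biome ambient scare one-shots: ghost voices, distant howls, ghostly swells and manifestations, spooky night sounds, lava ghosts and breathy moans.
# All quality levels share one weighted random pick among those categories, each nesting further random picks of waves balanced by mixers.
# Modulators vary some branches, while envelopers bend pitch downward over time or swell volume.
# Playback uses distant-creature attenuation, two-voice concurrency and a distance-scaled distant reverb send.
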SoundCue /Game/Sounds/Helios/Biomes/Underworld/Spots/Spot_Underworld_All_Scares_Cue
root=QualityLevel_0
nodes [
 {
  id=Enveloper_0
  children [Modulator_0]
  properties {pitchCurve {editorCurveData {keys [{time=0}, {time=1}, <at 7.5 0.8>]}}, volumeCurve {editorCurveData {keys [<at 0 1>]}}}
 }
 {
  id=Enveloper_1
  children [Mixer_2]
  properties {pitchCurve {editorCurveData {keys [{time=0}, {time=2}, <at 10 0.85>]}}, volumeCurve {editorCurveData {keys [<at 0 1>]}}}
 }
 {
  id=Enveloper_2
  children [WavePlayer_1]
  properties {pitchCurve {editorCurveData {keys [{time=0}]}}, volumeCurve {editorCurveData {keys [<at 0 0.5>,<at 2 0.5>,<at 3 2>]}}}
 }
 {
  id=Enveloper_3
  children [WavePlayer_12]
  properties {pitchCurve {editorCurveData {keys [{time=0}]}}, volumeCurve {editorCurveData {keys [<at 0 0.5>,<at 2 0.5>,<at 3 1.5>]}}}
 }
 {
  id=Enveloper_4
  children [Modulator_1]
  properties {pitchCurve {editorCurveData {keys [{time=0}, {time=1}, <at 7.5 0.8>]}}, volumeCurve {editorCurveData {keys [<at 0 1>]}}}
 }
 {
  id=Enveloper_5
  children [Modulator_2]
  properties {pitchCurve {editorCurveData {keys [{time=0}, {time=2}, <at 10 0.85>]}}, volumeCurve {editorCurveData {keys [<at 0 1>]}}}
 }
 {
  id=Enveloper_6
  children [Mixer_9]
  properties {pitchCurve {editorCurveData {keys [{time=0}, {time=2}, <at 10 0.85>]}}, volumeCurve {editorCurveData {keys [<at 0 1>]}}}
 }
 {
  id=Enveloper_7
  children [Mixer_12]
  properties {pitchCurve {editorCurveData {keys [{time=0}, {time=2}, <at 10 0.85>]}}, volumeCurve {editorCurveData {keys [<at 0 1>,<at 4 1>,<at 10 2>]}}}
 }
 {id=Mixer_0, children [WavePlayer_9], properties {inputVolume [0.8]}}
 {id=Mixer_1, children [WavePlayer_16], properties {inputVolume [0.1]}}
 {id=Mixer_10, children [WavePlayer_38], properties {inputVolume [2]}}
 {id=Mixer_11, children [WavePlayer_35], properties {inputVolume [0.75]}}
 {id=Mixer_12, children [Random_11], properties {inputVolume [4.4]}}
 {id=Mixer_2, children [WavePlayer_6], properties {inputVolume [1.5]}}
 {id=Mixer_3, children [WavePlayer_0], properties {inputVolume [0.8]}}
 {id=Mixer_4, children [WavePlayer_13], properties {inputVolume [0.75]}}
 {id=Mixer_5, children [Modulator_3], properties {inputVolume [0.5]}}
 {id=Mixer_6, children [WavePlayer_33], properties {inputVolume [0.75]}}
 {id=Mixer_7, children [Random_9], properties {inputVolume [1]}}
 {id=Mixer_8, children [Random_10], properties {inputVolume [0.75]}}
 {id=Mixer_9, children [Random_4], properties {inputVolume [0.75]}}
 {id=Modulator_0, children [Random_2], properties {pitchMax=1.1, pitchMin=0.7}}
 {id=Modulator_1, children [Random_6], properties {pitchMin=0.8, volumeMin=0.8}}
 {id=Modulator_2, children [Random_7]}
 {id=Modulator_3, children [WavePlayer_32], properties {pitchMax=0.9, pitchMin=0.8}}
 {id=Modulator_4, children [Mixer_8], properties {volumeMax=1.2}}
 {id=QualityLevel_0, children [Random_5,Random_5,Random_5,Random_5]}
 {
  id=Random_0
  children [WavePlayer_2,WavePlayer_3,WavePlayer_4,WavePlayer_5,Enveloper_1,WavePlayer_7,Mixer_1]
  properties {weights [1,1,0.25,0.25,1,1,0.2]}
 }
 {id=Random_1, children [Mixer_3, WavePlayer_8, Mixer_0, WavePlayer_10, WavePlayer_11, WavePlayer_17], properties {weights [1,1,1,1,1,1]}}
 {
  id=Random_10
  children [
   WavePlayer_45
   WavePlayer_46
   WavePlayer_47
   WavePlayer_48
   WavePlayer_49
   WavePlayer_50
   WavePlayer_51
   WavePlayer_52
   WavePlayer_53
   WavePlayer_54
   WavePlayer_55
  ]
  properties {weights [1,1,1,1,1,1,1,1,1,1,1]}
 }
 {id=Random_11, children [WavePlayer_56, WavePlayer_57, WavePlayer_58, WavePlayer_59, WavePlayer_60], properties {weights [1,1,1,1,1]}}
 {id=Random_2, children [Enveloper_2,Enveloper_3,Mixer_4], properties {weights [1,1,1]}}
 {id=Random_3, children [Random_0,Random_1], properties {weights [1,1]}}
 {id=Random_4, children [WavePlayer_34, Mixer_11, WavePlayer_36, WavePlayer_37, Mixer_10], properties {weights [1,1,1,1,1]}}
 {
  id=Random_5
  children [Random_3,Enveloper_0,Enveloper_4,Enveloper_5,Enveloper_6,Mixer_7,Modulator_4,Enveloper_7]
  properties {weights [1,1,0.8,1,1,0.5,0.8,1.2]}
 }
 {
  id=Random_6
  children [
   WavePlayer_14
   WavePlayer_15
   WavePlayer_18
   WavePlayer_19
   WavePlayer_20
   WavePlayer_21
   WavePlayer_22
   WavePlayer_23
   WavePlayer_24
   WavePlayer_25
   WavePlayer_26
  ]
  properties {weights [1,1,1,1,1,1,1,1,1,1,1]}
 }
 {
  id=Random_7
  children [WavePlayer_27, WavePlayer_28, WavePlayer_29, WavePlayer_30, WavePlayer_31, Mixer_5, Mixer_6]
  properties {weights [1,1,1,1,1,1,1]}
 }
 {
  id=Random_9
  children [WavePlayer_39, WavePlayer_40, WavePlayer_41, WavePlayer_42, WavePlayer_43, WavePlayer_44]
  properties {weights [1,1,1,1,1,1]}
 }
 {id=WavePlayer_0, properties {soundWaveAssetPtr=/Game/Sounds/Fort_Starlight/S1/SoundWaves/SpotFX_Starlight_Crypt_Ghost_Voice_01}}
 {id=WavePlayer_1, properties {soundWaveAssetPtr=/Game/Sounds/Creative/Ambiences/AbstractVoices/Voice_HowlDistantB001_Stereo}}
 {id=WavePlayer_10, properties {soundWaveAssetPtr=/Game/Sounds/Fort_Starlight/S1/SoundWaves/SpotFX_Starlight_Crypt_Ghost_Voice_04}}
 {id=WavePlayer_11, properties {soundWaveAssetPtr=/Game/Sounds/Fort_Starlight/S1/SoundWaves/SpotFX_Starlight_Crypt_Ghost_Voice_05}}
 {id=WavePlayer_12, properties {soundWaveAssetPtr=/Game/Sounds/Creative/Ambiences/AbstractVoices/Voice_HowlDistantB004_Stereo}}
 {id=WavePlayer_13, properties {soundWaveAssetPtr=/Game/Sounds/Creative/Ambiences/AbstractVoices/Voice_HowlDistantB006_Stereo}}
 {id=WavePlayer_14, properties {soundWaveAssetPtr=/Game/Sounds/Artemis/Biomes/Chrome/Spots/Spot_Chrome_Vocal_Swells_Ghostly_01}}
 {id=WavePlayer_15, properties {soundWaveAssetPtr=/Game/Sounds/Artemis/Biomes/Chrome/Spots/Spot_Chrome_Vocal_Swells_Ghostly_02}}
 {id=WavePlayer_16, properties {soundWaveAssetPtr=/Game/Sounds/Fort_Starlight/S1/SoundWaves/SpotFX_Starlight_Creepy_Rand_Night_03}}
 {id=WavePlayer_17, properties {soundWaveAssetPtr=/Game/Sounds/Fort_Starlight/S1/SoundWaves/SpotFX_Starlight_Crypt_Ghost_Voice_06}}
 {id=WavePlayer_18, properties {soundWaveAssetPtr=/Game/Sounds/Artemis/Biomes/Chrome/Spots/Spot_Chrome_Vocal_Swells_Ghostly_03}}
 {id=WavePlayer_19, properties {soundWaveAssetPtr=/Game/Sounds/Artemis/Biomes/Chrome/Spots/Spot_Chrome_Vocal_Swells_Ghostly_04}}
 {id=WavePlayer_2, properties {soundWaveAssetPtr=/Game/Sounds/Fort_Starlight/S1/SoundWaves/SpotFX_Starlight_Spooky_Rand_Night_02}}
 {id=WavePlayer_20, properties {soundWaveAssetPtr=/Game/Sounds/Artemis/Biomes/Chrome/Spots/Spot_Chrome_Vocal_Swells_Ghostly_05}}
 {id=WavePlayer_21, properties {soundWaveAssetPtr=/Game/Sounds/Artemis/Biomes/Chrome/Spots/Spot_Chrome_Vocal_Swells_Ghostly_06}}
 {id=WavePlayer_22, properties {soundWaveAssetPtr=/Game/Sounds/Artemis/Biomes/Chrome/Spots/Spot_Chrome_Vocal_Swells_Ghostly_07}}
 {id=WavePlayer_23, properties {soundWaveAssetPtr=/Game/Sounds/Artemis/Biomes/Chrome/Spots/Spot_Chrome_Vocal_Swells_Ghostly_08}}
 {id=WavePlayer_24, properties {soundWaveAssetPtr=/Game/Sounds/Artemis/Biomes/Chrome/Spots/Spot_Chrome_Vocal_Swells_Ghostly_09}}
 {id=WavePlayer_25, properties {soundWaveAssetPtr=/Game/Sounds/Artemis/Biomes/Chrome/Spots/Spot_Chrome_Vocal_Swells_Ghostly_10}}
 {id=WavePlayer_26, properties {soundWaveAssetPtr=/Game/Sounds/Artemis/Biomes/Chrome/Spots/Spot_Chrome_Vocal_Swells_Ghostly_11}}
 {id=WavePlayer_27, properties {soundWaveAssetPtr=&/Spot_Underworld_Ghostly_Manifestation_01}}
 {id=WavePlayer_28, properties {soundWaveAssetPtr=&/Spot_Underworld_Ghostly_Manifestation_02}}
 {id=WavePlayer_29, properties {soundWaveAssetPtr=&/Spot_Underworld_Ghostly_Manifestation_03}}
 {id=WavePlayer_3, properties {soundWaveAssetPtr=/Game/Sounds/Fort_Starlight/S1/SoundWaves/SpotFX_Starlight_Spooky_Rand_Night_04}}
 {id=WavePlayer_30, properties {soundWaveAssetPtr=&/Spot_Underworld_Ghostly_Manifestation_04}}
 {id=WavePlayer_31, properties {soundWaveAssetPtr=&/Spot_Underworld_Ghostly_Manifestation_05}}
 {id=WavePlayer_32, properties {soundWaveAssetPtr=&/Spot_Underworld_Ghostly_Manifestation_06}}
 {id=WavePlayer_33, properties {soundWaveAssetPtr=&/Spot_Underworld_Ghostly_Manifestation_07}}
 {id=WavePlayer_34, properties {soundWaveAssetPtr=&SpotFX_Starlight_Spooky_Rand_Night_02}}
 {id=WavePlayer_35, properties {soundWaveAssetPtr=&SpotFX_Starlight_Spooky_Rand_Night_04}}
 {id=WavePlayer_36, properties {soundWaveAssetPtr=/Game/Sounds/Fort_Starlight/S1/SoundWaves/SpotFX_Starlight_Spooky_Rand_Night_05}}
 {id=WavePlayer_37, properties {soundWaveAssetPtr=/Game/Sounds/Fort_Starlight/S1/SoundWaves/SpotFX_Starlight_Spooky_Rand_Night_06}}
 {id=WavePlayer_38, properties {soundWaveAssetPtr=/Game/Sounds/Fort_Starlight/S1/SoundWaves/SpotFX_Starlight_Spooky_Rand_Night_03}}
 {id=WavePlayer_39, properties {soundWaveAssetPtr=&SpotFX_Starlight_Crypt_Ghost_Voice_01}}
 {id=WavePlayer_4, properties {soundWaveAssetPtr=&SpotFX_Starlight_Spooky_Rand_Night_05}}
 {id=WavePlayer_40, properties {soundWaveAssetPtr=/Game/Sounds/Fort_Starlight/S1/SoundWaves/SpotFX_Starlight_Crypt_Ghost_Voice_02}}
 {id=WavePlayer_41, properties {soundWaveAssetPtr=/Game/Sounds/Fort_Starlight/S1/SoundWaves/SpotFX_Starlight_Crypt_Ghost_Voice_03}}
 {id=WavePlayer_42, properties {soundWaveAssetPtr=&SpotFX_Starlight_Crypt_Ghost_Voice_04}}
 {id=WavePlayer_43, properties {soundWaveAssetPtr=&SpotFX_Starlight_Crypt_Ghost_Voice_05}}
 {id=WavePlayer_44, properties {soundWaveAssetPtr=&SpotFX_Starlight_Crypt_Ghost_Voice_06}}
 {id=WavePlayer_45, properties {soundWaveAssetPtr=/Game/Sounds/Fort_Starlight/S2/SoundWaves/SpotFX_Starlight_S2_Amb_Lava_Ghost_01}}
 {id=WavePlayer_46, properties {soundWaveAssetPtr=/Game/Sounds/Fort_Starlight/S2/SoundWaves/SpotFX_Starlight_S2_Amb_Lava_Ghost_07}}
 {id=WavePlayer_47, properties {soundWaveAssetPtr=/Game/Sounds/Fort_Starlight/S2/SoundWaves/SpotFX_Starlight_S2_Amb_Lava_Ghost_09}}
 {id=WavePlayer_48, properties {soundWaveAssetPtr=/Game/Sounds/Fort_Starlight/S2/SoundWaves/SpotFX_Starlight_S2_Amb_Lava_Ghost_10}}
 {id=WavePlayer_49, properties {soundWaveAssetPtr=/Game/Sounds/Fort_Starlight/S2/SoundWaves/SpotFX_Starlight_S2_Amb_Lava_Ghost_11}}
 {id=WavePlayer_5, properties {soundWaveAssetPtr=&SpotFX_Starlight_Spooky_Rand_Night_06}}
 {id=WavePlayer_50, properties {soundWaveAssetPtr=/Game/Sounds/Fort_Starlight/S2/SoundWaves/SpotFX_Starlight_S2_Amb_Lava_Ghost_02}}
 {id=WavePlayer_51, properties {soundWaveAssetPtr=/Game/Sounds/Fort_Starlight/S2/SoundWaves/SpotFX_Starlight_S2_Amb_Lava_Ghost_03}}
 {id=WavePlayer_52, properties {soundWaveAssetPtr=/Game/Sounds/Fort_Starlight/S2/SoundWaves/SpotFX_Starlight_S2_Amb_Lava_Ghost_04}}
 {id=WavePlayer_53, properties {soundWaveAssetPtr=/Game/Sounds/Fort_Starlight/S2/SoundWaves/SpotFX_Starlight_S2_Amb_Lava_Ghost_05}}
 {id=WavePlayer_54, properties {soundWaveAssetPtr=/Game/Sounds/Fort_Starlight/S2/SoundWaves/SpotFX_Starlight_S2_Amb_Lava_Ghost_06}}
 {id=WavePlayer_55, properties {soundWaveAssetPtr=/Game/Sounds/Fort_Starlight/S2/SoundWaves/SpotFX_Starlight_S2_Amb_Lava_Ghost_08}}
 {id=WavePlayer_56, properties {soundWaveAssetPtr=/Game/Sounds/Fort_Starlight/S1/SoundWaves/SpotFX_Starlight_Creepy_Breathy_Moan_St_01}}
 {id=WavePlayer_57, properties {soundWaveAssetPtr=/Game/Sounds/Fort_Starlight/S1/SoundWaves/SpotFX_Starlight_Creepy_Breathy_Moan_St_02}}
 {id=WavePlayer_58, properties {soundWaveAssetPtr=/Game/Sounds/Fort_Starlight/S1/SoundWaves/SpotFX_Starlight_Creepy_Breathy_Moan_St_03}}
 {id=WavePlayer_59, properties {soundWaveAssetPtr=/Game/Sounds/Fort_Starlight/S1/SoundWaves/SpotFX_Starlight_Creepy_Breathy_Moan_St_04}}
 {id=WavePlayer_6, properties {soundWaveAssetPtr=&SpotFX_Starlight_Spooky_Rand_Night_03}}
 {id=WavePlayer_60, properties {soundWaveAssetPtr=/Game/Sounds/Fort_Starlight/S1/SoundWaves/SpotFX_Starlight_Creepy_Breathy_Moan_St_05}}
 {id=WavePlayer_7, properties {soundWaveAssetPtr=/Game/Sounds/Fort_Starlight/S1/SoundWaves/SpotFX_Starlight_Creepy_Rand_Day_01}}
 {id=WavePlayer_8, properties {soundWaveAssetPtr=&SpotFX_Starlight_Crypt_Ghost_Voice_02}}
 {id=WavePlayer_9, properties {soundWaveAssetPtr=&SpotFX_Starlight_Crypt_Ghost_Voice_03}}
]
properties {
 attenuationSettings=/Game/Sounds/Attenuation/Fort_TileAmbient_OneShot_Distant_Creature_Att
 bEnableBusSends=false
 bOverrideConcurrency=true
 maxDistance=8500
 soundClassObject=/Game/Sounds/FortSoundClassesAndModes/Fort_Ambient_OneShots_Indoors
 volumeMultiplier=0.401
 attenuationOverrides {bAttenuate=false}
 concurrencyOverrides {maxCount=2, voiceStealReleaseTime=2}
 soundSubmixSends [{
   maxSendDistance=8000
   minSendLevel=0.25
   soundSubmix=&Fortnite_Reverb_Distant
   customSendLevelCurve {editorCurveData {keys [<at 0 0.25>,<at 8000 1>]}}
  }]
}
