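# Underworld biome ambient spot cue that plays a distant, creepy breathy drone moan as an indoor one-shot.
# An equal-weight Random node picks one of five Starlight moan waves and feeds a boosted Mixer, then an Enveloper shaping pitch and volume over time.
# Distant-creature attenuation and a distance-scaled send to the distant reverb submix place it far away, while a concurrency override allows only one voice.
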SoundCue /Game/Sounds/Helios/Biomes/Underworld/Spots/Spot_Underworld_Breathy_Drone_Moan_Cue
root=Enveloper_1
nodes [
 {
  id=Enveloper_1
  children [Mixer_1]
  properties {pitchCurve {editorCurveData {keys [{time=0}, {time=2}, <at 10 0.85>]}}, volumeCurve {editorCurveData {keys [<at 0 1>,<at 4 1>,<at 10 2>]}}}
 }
 {id=Mixer_1, children [Random_1], properties {inputVolume [4.4]}}
 {id=Random_1, children [WavePlayer_0, WavePlayer_8, WavePlayer_9, WavePlayer_10, WavePlayer_11], properties {weights [1,1,1,1,1]}}
 {id=WavePlayer_0, properties {soundWaveAssetPtr=/Game/Sounds/Fort_Starlight/S1/SoundWaves/SpotFX_Starlight_Creepy_Breathy_Moan_St_01}}
 {id=WavePlayer_10, properties {soundWaveAssetPtr=/Game/Sounds/Fort_Starlight/S1/SoundWaves/SpotFX_Starlight_Creepy_Breathy_Moan_St_04}}
 {id=WavePlayer_11, properties {soundWaveAssetPtr=/Game/Sounds/Fort_Starlight/S1/SoundWaves/SpotFX_Starlight_Creepy_Breathy_Moan_St_05}}
 {id=WavePlayer_8, properties {soundWaveAssetPtr=/Game/Sounds/Fort_Starlight/S1/SoundWaves/SpotFX_Starlight_Creepy_Breathy_Moan_St_02}}
 {id=WavePlayer_9, properties {soundWaveAssetPtr=/Game/Sounds/Fort_Starlight/S1/SoundWaves/SpotFX_Starlight_Creepy_Breathy_Moan_St_03}}
]
properties {
 attenuationSettings=/Game/Sounds/Attenuation/Fort_TileAmbient_OneShot_Distant_Creature_Att
 bOverrideConcurrency=true
 maxDistance=8500
 soundClassObject=/Game/Sounds/FortSoundClassesAndModes/Fort_Ambient_OneShots_Indoors
 volumeMultiplier=0.501
 concurrencyOverrides {maxCount=1, voiceStealReleaseTime=5}
 soundSubmixSends [{
   maxSendDistance=8000
   minSendLevel=0.25
   soundSubmix=&Fortnite_Reverb_Distant
   customSendLevelCurve {editorCurveData {keys [<at 0 0.25>,<at 8000 1.2>]}}
  }]
}
