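# Sound cue for desolate-wind evening one-shot spot ambience in the Underworld biome, reusing Alpine night one-shot recordings.
# A quality-level switch feeds the two higher-listed slots from quad waves and the other two from stereo waves.
# Each branch randomly picks one of four equally weighted waves, then modulates pitch and volume slightly.
# The cue plays quietly through outdoor ambient sound class with tile one-shot attenuation.
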SoundCue /Game/Sounds/Helios/Biomes/Underworld/Spots/Spot_Underworld_Desolate_Wind_Oneshots_Evening_Cue
root=QualityLevel_0
nodes [
 {id=Modulator_0, children [Random_3], properties {pitchMax=1.1, pitchMin=0.9, volumeMax=0.9, volumeMin=0.8}}
 {id=Modulator_1, children [Random_2], properties {pitchMax=1.1, pitchMin=0.9, volumeMax=1.1, volumeMin=0.9}}
 {id=QualityLevel_0, children [Modulator_1,Modulator_1,Modulator_0,Modulator_0]}
 {id=Random_2, children [WavePlayer_0, WavePlayer_9, WavePlayer_10, WavePlayer_11], properties {weights [1,1,1,1]}}
 {id=Random_3, children [WavePlayer_12, WavePlayer_13, WavePlayer_14, WavePlayer_15], properties {weights [1,1,1,1]}}
 {id=WavePlayer_0, properties {soundWaveAssetPtr=/Game/Sounds/Apollo/Biomes/Alpine/Biome_Alpine_Oneshot_Night_Quad_01}}
 {id=WavePlayer_10, properties {soundWaveAssetPtr=/Game/Sounds/Apollo/Biomes/Alpine/Biome_Alpine_Oneshot_Night_Quad_03}}
 {id=WavePlayer_11, properties {soundWaveAssetPtr=/Game/Sounds/Apollo/Biomes/Alpine/Biome_Alpine_Oneshot_Night_Quad_04}}
 {id=WavePlayer_12, properties {soundWaveAssetPtr=/Game/Sounds/Apollo/Biomes/Alpine/Biome_Alpine_Oneshot_Night_ST_01}}
 {id=WavePlayer_13, properties {soundWaveAssetPtr=/Game/Sounds/Apollo/Biomes/Alpine/Biome_Alpine_Oneshot_Night_ST_02}}
 {id=WavePlayer_14, properties {soundWaveAssetPtr=/Game/Sounds/Apollo/Biomes/Alpine/Biome_Alpine_Oneshot_Night_ST_03}}
 {id=WavePlayer_15, properties {soundWaveAssetPtr=/Game/Sounds/Apollo/Biomes/Alpine/Biome_Alpine_Oneshot_Night_ST_04}}
 {id=WavePlayer_9, properties {soundWaveAssetPtr=/Game/Sounds/Apollo/Biomes/Alpine/Biome_Alpine_Oneshot_Night_Quad_02}}
]
properties {
 attenuationSettings=/Game/Sounds/Attenuation/Fort_TileAmbient_OneShot_Att
 maxDistance=8000
 soundClassObject=&Fort_Ambient_Outdoors_Athena
 volumeMultiplier=0.125
 preEffectBusSends [{}]
}
