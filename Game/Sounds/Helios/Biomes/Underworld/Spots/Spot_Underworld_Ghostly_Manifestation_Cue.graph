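# Ghostly manifestation one-shot ambient spot cue for the Helios Underworld biome, using distant-creature attenuation and an exterior ambient bus send.
# A Random node picks among seven equally weighted waves, with two reduced in volume and one of those pitched down.
# The pick passes through a Modulator and a pitch-curve Enveloper into a Quality Level root, which leaves two of four slots empty.
# A concurrency override limits simultaneous voices.
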SoundCue /Game/Sounds/Helios/Biomes/Underworld/Spots/Spot_Underworld_Ghostly_Manifestation_Cue
root=QualityLevel_0
nodes [
 {
  id=Enveloper_1
  children [Modulator_1]
  properties {pitchCurve {editorCurveData {keys [{time=0}, {time=2}, <at 10 0.85>]}}, volumeCurve {editorCurveData {keys [<at 0 1>]}}}
 }
 {id=Mixer_0, children [Modulator_2], properties {inputVolume [0.5]}}
 {id=Mixer_3, children [WavePlayer_6], properties {inputVolume [0.75]}}
 {id=Modulator_1, children [Random_0]}
 {id=Modulator_2, children [WavePlayer_5], properties {pitchMax=0.9, pitchMin=0.8}}
 {id=QualityLevel_0, children [Enveloper_1,Enveloper_1,null,null]}
 {
  id=Random_0
  children [WavePlayer_0,WavePlayer_1,WavePlayer_2,WavePlayer_3,WavePlayer_4,Mixer_0,Mixer_3]
  properties {weights [1,1,1,1,1,1,1]}
 }
 {id=WavePlayer_0, properties {soundWaveAssetPtr=&/Spot_Underworld_Ghostly_Manifestation_01}}
 {id=WavePlayer_1, properties {soundWaveAssetPtr=&/Spot_Underworld_Ghostly_Manifestation_02}}
 {id=WavePlayer_2, properties {soundWaveAssetPtr=&/Spot_Underworld_Ghostly_Manifestation_03}}
 {id=WavePlayer_3, properties {soundWaveAssetPtr=&/Spot_Underworld_Ghostly_Manifestation_04}}
 {id=WavePlayer_4, properties {soundWaveAssetPtr=&/Spot_Underworld_Ghostly_Manifestation_05}}
 {id=WavePlayer_5, properties {soundWaveAssetPtr=&/Spot_Underworld_Ghostly_Manifestation_06}}
 {id=WavePlayer_6, properties {soundWaveAssetPtr=&/Spot_Underworld_Ghostly_Manifestation_07}}
]
properties {
 attenuationSettings=/Game/Sounds/Attenuation/Fort_TileAmbient_OneShot_Distant_Creature_Att
 bOverrideConcurrency=true
 maxDistance=8500
 soundClassObject=/Game/Sounds/FortSoundClassesAndModes/Fort_Ambient_OneShots_Indoors
 volumeMultiplier=0.501
 busSends [{audioBus=null, minSendLevel=1, sendLevel=1, soundSourceBus=&AmbientAudio_Exterior_SourceBus}]
 concurrencyOverrides {maxCount=2, voiceStealReleaseTime=2}
}
