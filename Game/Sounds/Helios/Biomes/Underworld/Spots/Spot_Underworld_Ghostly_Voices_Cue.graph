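# Sound cue for distant one-shot ghostly voice spot ambience in the Helios Underworld biome, reusing Starlight sound waves.
# A weighted random node picks spooky night sounds through a pitch-dropping Enveloper, crypt ghost voices, or lava ghost sounds through a Modulator.
# The root quality switch feeds that random node only in its first two slots, leaving the others empty.
# Playback uses distant-creature attenuation, a tight concurrency override and an exterior ambient bus send.
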SoundCue /Game/Sounds/Helios/Biomes/Underworld/Spots/Spot_Underworld_Ghostly_Voices_Cue
root=QualityLevel_0
nodes [
 {
  id=Enveloper_0
  children [Mixer_1]
  properties {pitchCurve {editorCurveData {keys [{time=0}, {time=2}, <at 10 0.85>]}}, volumeCurve {editorCurveData {keys [<at 0 1>]}}}
 }
 {id=Mixer_1, children [Random_4], properties {inputVolume [0.75]}}
 {id=Mixer_2, children [WavePlayer_32], properties {inputVolume [2]}}
 {id=Mixer_4, children [Random_7], properties {inputVolume [1]}}
 {id=Mixer_5, children [Random_8], properties {inputVolume [0.75]}}
 {id=Mixer_8, children [WavePlayer_29], properties {inputVolume [0.75]}}
 {id=Modulator_0, children [Mixer_5], properties {volumeMax=1.2}}
 {id=QualityLevel_0, children [Random_2,Random_2,null,null]}
 {id=Random_2, children [Enveloper_0,Mixer_4,Modulator_0], properties {weights [1,0.5,0.8]}}
 {id=Random_4, children [WavePlayer_28, Mixer_8, WavePlayer_30, WavePlayer_31, Mixer_2], properties {weights [1,1,1,1,1]}}
 {
  id=Random_7
  children [WavePlayer_39, WavePlayer_40, WavePlayer_41, WavePlayer_42, WavePlayer_43, WavePlayer_44]
  properties {weights [1,1,1,1,1,1]}
 }
 {
  id=Random_8
  children [
   WavePlayer_45
   WavePlayer_46
   WavePlayer_47
   WavePlayer_48
   WavePlayer_49
   WavePlayer_50
   WavePlayer_51
   WavePlayer_52
   WavePlayer_53
   WavePlayer_54
   WavePlayer_55
  ]
  properties {weights [1,1,1,1,1,1,1,1,1,1,1]}
 }
 {id=WavePlayer_28, properties {soundWaveAssetPtr=/Game/Sounds/Fort_Starlight/S1/SoundWaves/SpotFX_Starlight_Spooky_Rand_Night_02}}
 {id=WavePlayer_29, properties {soundWaveAssetPtr=/Game/Sounds/Fort_Starlight/S1/SoundWaves/SpotFX_Starlight_Spooky_Rand_Night_04}}
 {id=WavePlayer_30, properties {soundWaveAssetPtr=/Game/Sounds/Fort_Starlight/S1/SoundWaves/SpotFX_Starlight_Spooky_Rand_Night_05}}
 {id=WavePlayer_31, properties {soundWaveAssetPtr=/Game/Sounds/Fort_Starlight/S1/SoundWaves/SpotFX_Starlight_Spooky_Rand_Night_06}}
 {id=WavePlayer_32, properties {soundWaveAssetPtr=/Game/Sounds/Fort_Starlight/S1/SoundWaves/SpotFX_Starlight_Spooky_Rand_Night_03}}
 {id=WavePlayer_39, properties {soundWaveAssetPtr=/Game/Sounds/Fort_Starlight/S1/SoundWaves/SpotFX_Starlight_Crypt_Ghost_Voice_01}}
 {id=WavePlayer_40, properties {soundWaveAssetPtr=/Game/Sounds/Fort_Starlight/S1/SoundWaves/SpotFX_Starlight_Crypt_Ghost_Voice_02}}
 {id=WavePlayer_41, properties {soundWaveAssetPtr=/Game/Sounds/Fort_Starlight/S1/SoundWaves/SpotFX_Starlight_Crypt_Ghost_Voice_03}}
 {id=WavePlayer_42, properties {soundWaveAssetPtr=/Game/Sounds/Fort_Starlight/S1/SoundWaves/SpotFX_Starlight_Crypt_Ghost_Voice_04}}
 {id=WavePlayer_43, properties {soundWaveAssetPtr=/Game/Sounds/Fort_Starlight/S1/SoundWaves/SpotFX_Starlight_Crypt_Ghost_Voice_05}}
 {id=WavePlayer_44, properties {soundWaveAssetPtr=/Game/Sounds/Fort_Starlight/S1/SoundWaves/SpotFX_Starlight_Crypt_Ghost_Voice_06}}
 {id=WavePlayer_45, properties {soundWaveAssetPtr=/Game/Sounds/Fort_Starlight/S2/SoundWaves/SpotFX_Starlight_S2_Amb_Lava_Ghost_01}}
 {id=WavePlayer_46, properties {soundWaveAssetPtr=/Game/Sounds/Fort_Starlight/S2/SoundWaves/SpotFX_Starlight_S2_Amb_Lava_Ghost_07}}
 {id=WavePlayer_47, properties {soundWaveAssetPtr=/Game/Sounds/Fort_Starlight/S2/SoundWaves/SpotFX_Starlight_S2_Amb_Lava_Ghost_09}}
 {id=WavePlayer_48, properties {soundWaveAssetPtr=/Game/Sounds/Fort_Starlight/S2/SoundWaves/SpotFX_Starlight_S2_Amb_Lava_Ghost_10}}
 {id=WavePlayer_49, properties {soundWaveAssetPtr=/Game/Sounds/Fort_Starlight/S2/SoundWaves/SpotFX_Starlight_S2_Amb_Lava_Ghost_11}}
 {id=WavePlayer_50, properties {soundWaveAssetPtr=/Game/Sounds/Fort_Starlight/S2/SoundWaves/SpotFX_Starlight_S2_Amb_Lava_Ghost_02}}
 {id=WavePlayer_51, properties {soundWaveAssetPtr=/Game/Sounds/Fort_Starlight/S2/SoundWaves/SpotFX_Starlight_S2_Amb_Lava_Ghost_03}}
 {id=WavePlayer_52, properties {soundWaveAssetPtr=/Game/Sounds/Fort_Starlight/S2/SoundWaves/SpotFX_Starlight_S2_Amb_Lava_Ghost_04}}
 {id=WavePlayer_53, properties {soundWaveAssetPtr=/Game/Sounds/Fort_Starlight/S2/SoundWaves/SpotFX_Starlight_S2_Amb_Lava_Ghost_05}}
 {id=WavePlayer_54, properties {soundWaveAssetPtr=/Game/Sounds/Fort_Starlight/S2/SoundWaves/SpotFX_Starlight_S2_Amb_Lava_Ghost_06}}
 {id=WavePlayer_55, properties {soundWaveAssetPtr=/Game/Sounds/Fort_Starlight/S2/SoundWaves/SpotFX_Starlight_S2_Amb_Lava_Ghost_08}}
]
properties {
 attenuationSettings=/Game/Sounds/Attenuation/Fort_TileAmbient_OneShot_Distant_Creature_Att
 bOverrideConcurrency=true
 maxDistance=8500
 soundClassObject=/Game/Sounds/FortSoundClassesAndModes/Fort_Ambient_OneShots_Indoors
 volumeMultiplier=0.501
 busSends [{audioBus=null, minSendLevel=1, sendLevel=1, soundSourceBus=&AmbientAudio_Exterior_SourceBus}]
 concurrencyOverrides {maxCount=2, voiceStealReleaseTime=2}
}
