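# Sound cue for random musical one-shot ambient spots in the Underworld biome, named for its Castle area.
# A top-level Random picks evenly between two nested Randoms, one over ten general Underworld one-shots and one over ten Cerberus variants, each equally weighted.
# It plays in the Music sound class at reduced volume with no-reverb ambient attenuation.
# A concurrency override allows one instance and blocks new ones while it plays.
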SoundCue /Game/Sounds/Helios/Biomes/Underworld/Spots/Spot_Underworld_Musical_Oneshot_Castle_Cue
root=Random_2
nodes [
 {
  id=Random_0
  children [WavePlayer_0,WavePlayer_1,WavePlayer_2,WavePlayer_3,WavePlayer_4,WavePlayer_5,WavePlayer_6,WavePlayer_7,WavePlayer_8,WavePlayer_9]
  properties {weights [1,1,1,1,1,1,1,1,1,1]}
 }
 {
  id=Random_1
  children [
   WavePlayer_10
   WavePlayer_11
   WavePlayer_12
   WavePlayer_13
   WavePlayer_14
   WavePlayer_15
   WavePlayer_16
   WavePlayer_17
   WavePlayer_18
   WavePlayer_19
  ]
  properties {weights [1,1,1,1,1,1,1,1,1,1]}
 }
 {id=Random_2, children [Random_1,Random_0], properties {weights [1,1]}}
 {id=WavePlayer_0, properties {soundWaveAssetPtr=&/Spot_Underworld_Musical_Oneshot_Cerberus_01}}
 {id=WavePlayer_1, properties {soundWaveAssetPtr=&/Spot_Underworld_Musical_Oneshot_Cerberus_02}}
 {id=WavePlayer_10, properties {soundWaveAssetPtr=&/Spot_Underworld_Musical_Oneshot_10}}
 {id=WavePlayer_11, properties {soundWaveAssetPtr=&/Spot_Underworld_Musical_Oneshot_09}}
 {id=WavePlayer_12, properties {soundWaveAssetPtr=&/Spot_Underworld_Musical_Oneshot_08}}
 {id=WavePlayer_13, properties {soundWaveAssetPtr=&/Spot_Underworld_Musical_Oneshot_07}}
 {id=WavePlayer_14, properties {soundWaveAssetPtr=&/Spot_Underworld_Musical_Oneshot_06}}
 {id=WavePlayer_15, properties {soundWaveAssetPtr=&/Spot_Underworld_Musical_Oneshot_05}}
 {id=WavePlayer_16, properties {soundWaveAssetPtr=&/Spot_Underworld_Musical_Oneshot_04}}
 {id=WavePlayer_17, properties {soundWaveAssetPtr=&/Spot_Underworld_Musical_Oneshot_03}}
 {id=WavePlayer_18, properties {soundWaveAssetPtr=&/Spot_Underworld_Musical_Oneshot_02}}
 {id=WavePlayer_19, properties {soundWaveAssetPtr=&/Spot_Underworld_Musical_Oneshot_01}}
 {id=WavePlayer_2, properties {soundWaveAssetPtr=&/Spot_Underworld_Musical_Oneshot_Cerberus_03}}
 {id=WavePlayer_3, properties {soundWaveAssetPtr=&/Spot_Underworld_Musical_Oneshot_Cerberus_04}}
 {id=WavePlayer_4, properties {soundWaveAssetPtr=&/Spot_Underworld_Musical_Oneshot_Cerberus_05}}
 {id=WavePlayer_5, properties {soundWaveAssetPtr=&/Spot_Underworld_Musical_Oneshot_Cerberus_06}}
 {id=WavePlayer_6, properties {soundWaveAssetPtr=&/Spot_Underworld_Musical_Oneshot_Cerberus_07}}
 {id=WavePlayer_7, properties {soundWaveAssetPtr=&/Spot_Underworld_Musical_Oneshot_Cerberus_08}}
 {id=WavePlayer_8, properties {soundWaveAssetPtr=&/Spot_Underworld_Musical_Oneshot_Cerberus_09}}
 {id=WavePlayer_9, properties {soundWaveAssetPtr=&/Spot_Underworld_Musical_Oneshot_Cerberus_10}}
]
properties {
 attenuationSettings=/Game/Sounds/Attenuation/Fort_Ambient_Athena_Musical_Oneshot_NoVerb_Att
 bOverrideConcurrency=true
 maxDistance=2097152
 soundClassObject=/Game/Sounds/FortSoundClassesAndModes/Fort_Music
 volumeMultiplier=0.55
 concurrencyOverrides {maxCount=1, resolutionRule=PreventNew, voiceStealReleaseTime=2}
}
