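# Sound cue for musical one-shot spot sounds in the Underworld biome of Helios.
# A Random node picks among ten equally weighted wave variations and is excluded from branch culling.
# It routes to the Fort_Music sound class at reduced volume, with an ambient no-verb musical one-shot attenuation asset.
# A concurrency override allows only one instance at a time, blocking new plays rather than stealing the active one.
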SoundCue /Game/Sounds/Helios/Biomes/Underworld/Spots/Spot_Underworld_Musical_Oneshot_Cue
root=Random_1
nodes [
 {
  id=Random_1
  children [
   WavePlayer_10
   WavePlayer_11
   WavePlayer_12
   WavePlayer_13
   WavePlayer_14
   WavePlayer_15
   WavePlayer_16
   WavePlayer_17
   WavePlayer_18
   WavePlayer_19
  ]
  properties {bShouldExcludeFromBranchCulling=true, weights [1,1,1,1,1,1,1,1,1,1]}
 }
 {id=WavePlayer_10, properties {soundWaveAssetPtr=&/Spot_Underworld_Musical_Oneshot_10}}
 {id=WavePlayer_11, properties {soundWaveAssetPtr=&/Spot_Underworld_Musical_Oneshot_09}}
 {id=WavePlayer_12, properties {soundWaveAssetPtr=&/Spot_Underworld_Musical_Oneshot_08}}
 {id=WavePlayer_13, properties {soundWaveAssetPtr=&/Spot_Underworld_Musical_Oneshot_07}}
 {id=WavePlayer_14, properties {soundWaveAssetPtr=&/Spot_Underworld_Musical_Oneshot_06}}
 {id=WavePlayer_15, properties {soundWaveAssetPtr=&/Spot_Underworld_Musical_Oneshot_05}}
 {id=WavePlayer_16, properties {soundWaveAssetPtr=&/Spot_Underworld_Musical_Oneshot_04}}
 {id=WavePlayer_17, properties {soundWaveAssetPtr=&/Spot_Underworld_Musical_Oneshot_03}}
 {id=WavePlayer_18, properties {soundWaveAssetPtr=&/Spot_Underworld_Musical_Oneshot_02}}
 {id=WavePlayer_19, properties {soundWaveAssetPtr=&/Spot_Underworld_Musical_Oneshot_01}}
]
properties {
 attenuationSettings=/Game/Sounds/Attenuation/Fort_Ambient_Athena_Musical_Oneshot_NoVerb_Att
 bOverrideConcurrency=true
 maxDistance=2097152
 soundClassObject=/Game/Sounds/FortSoundClassesAndModes/Fort_Music
 volumeMultiplier=0.55
 concurrencyOverrides {maxCount=1, resolutionRule=PreventNew, voiceStealReleaseTime=2}
}
