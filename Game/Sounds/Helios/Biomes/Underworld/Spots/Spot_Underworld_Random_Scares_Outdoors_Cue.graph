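# Sound cue for random outdoor scare one-shots in the Underworld biome, with distant-creature attenuation and a distance-scaled send to a distant reverb submix.
# A random node picks between spooky night sounds and crypt ghost voices, each pool holding five waves.
# One wave per pool passes through a mixer and an enveloper with a pitch curve.
# The quality switch leaves its last two levels empty, and a concurrency override caps simultaneous voices.
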SoundCue /Game/Sounds/Helios/Biomes/Underworld/Spots/Spot_Underworld_Random_Scares_Outdoors_Cue
root=QualityLevel_0
nodes [
 {
  id=Enveloper_0
  children [Mixer_0]
  properties {pitchCurve {editorCurveData {keys [{time=0}, {time=2}, <at 10 0.85>]}}, volumeCurve {editorCurveData {keys [<at 0 1>]}}}
 }
 {
  id=Enveloper_1
  children [Mixer_2]
  properties {pitchCurve {editorCurveData {keys [{time=0}, {time=2}, <at 10 0.85>]}}, volumeCurve {editorCurveData {keys [<at 0 1>]}}}
 }
 {id=Mixer_0, children [WavePlayer_9], properties {inputVolume [0.8]}}
 {id=Mixer_2, children [WavePlayer_6], properties {inputVolume [1.5]}}
 {id=QualityLevel_0, children [Random_3,Random_3,null,null]}
 {id=Random_0, children [WavePlayer_2,WavePlayer_3,WavePlayer_4,WavePlayer_5,Enveloper_1], properties {weights [1,1,0.25,0.25,1]}}
 {id=Random_1, children [WavePlayer_8, Enveloper_0, WavePlayer_10, WavePlayer_11, WavePlayer_17], properties {weights [1,1,1,1,1]}}
 {id=Random_3, children [Random_0,Random_1], properties {weights [1,1]}}
 {id=WavePlayer_10, properties {soundWaveAssetPtr=/Game/Sounds/Fort_Starlight/S1/SoundWaves/SpotFX_Starlight_Crypt_Ghost_Voice_04}}
 {id=WavePlayer_11, properties {soundWaveAssetPtr=/Game/Sounds/Fort_Starlight/S1/SoundWaves/SpotFX_Starlight_Crypt_Ghost_Voice_05}}
 {id=WavePlayer_17, properties {soundWaveAssetPtr=/Game/Sounds/Fort_Starlight/S1/SoundWaves/SpotFX_Starlight_Crypt_Ghost_Voice_06}}
 {id=WavePlayer_2, properties {soundWaveAssetPtr=/Game/Sounds/Fort_Starlight/S1/SoundWaves/SpotFX_Starlight_Spooky_Rand_Night_02}}
 {id=WavePlayer_3, properties {soundWaveAssetPtr=/Game/Sounds/Fort_Starlight/S1/SoundWaves/SpotFX_Starlight_Spooky_Rand_Night_04}}
 {id=WavePlayer_4, properties {soundWaveAssetPtr=/Game/Sounds/Fort_Starlight/S1/SoundWaves/SpotFX_Starlight_Spooky_Rand_Night_05}}
 {id=WavePlayer_5, properties {soundWaveAssetPtr=/Game/Sounds/Fort_Starlight/S1/SoundWaves/SpotFX_Starlight_Spooky_Rand_Night_06}}
 {id=WavePlayer_6, properties {soundWaveAssetPtr=/Game/Sounds/Fort_Starlight/S1/SoundWaves/SpotFX_Starlight_Spooky_Rand_Night_03}}
 {id=WavePlayer_8, properties {soundWaveAssetPtr=/Game/Sounds/Fort_Starlight/S1/SoundWaves/SpotFX_Starlight_Crypt_Ghost_Voice_02}}
 {id=WavePlayer_9, properties {soundWaveAssetPtr=/Game/Sounds/Fort_Starlight/S1/SoundWaves/SpotFX_Starlight_Crypt_Ghost_Voice_03}}
]
properties {
 attenuationSettings=/Game/Sounds/Attenuation/Fort_TileAmbient_OneShot_Distant_Creature_Att
 bEnableBusSends=false
 bOverrideConcurrency=true
 maxDistance=8500
 soundClassObject=&Fort_Ambient_OneShots
 volumeMultiplier=0.401
 attenuationOverrides {bAttenuate=false}
 concurrencyOverrides {maxCount=2, voiceStealReleaseTime=2}
 soundSubmixSends [{
   maxSendDistance=8000
   minSendLevel=0.25
   soundSubmix=&Fortnite_Reverb_Distant
   customSendLevelCurve {editorCurveData {keys [<at 0 0.25>,<at 8000 1>]}}
  }]
}
