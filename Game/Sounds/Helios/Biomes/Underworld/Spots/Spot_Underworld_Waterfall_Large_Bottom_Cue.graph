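# Looping spot ambience cue for the bottom of a large Underworld waterfall, built from reused Reality Falls and Athena water loops.
# A quality switch sends its first two levels to a shared two-layer mix: a distant roar plus a briefly delayed, louder bottom loop, under a sound class override.
# The last two levels each play one close roar loop instead.
# Every layer has randomized pitch and its own attenuation with listener focus, and concurrency caps instances.
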SoundCue /Game/Sounds/Helios/Biomes/Underworld/Spots/Spot_Underworld_Waterfall_Large_Bottom_Cue
root=QualityLevel_0
nodes [
 {
  id=Attenuation_2
  children [Modulator_4]
  properties {
   bOverrideAttenuation=true
   attenuationOverrides {
    bEnableFocusInterpolation=true
    bEnableListenerFocus=true
    bEnableOcclusion=true
    distanceAlgorithm=Custom
    falloffDistance=5000
    lPFFrequencyAtMax=1000
    lPFRadiusMax=4000
    lPFRadiusMin=1000
    nonFocusVolumeAttenuation=0.8
    nonSpatializedRadiusStart=5000
    occlusionInterpolationTime=3
    occlusionLowPassFilterFrequency=1000
    occlusionTraceChannel=ECC_GameTraceChannel11
    stereoSpread=10000
    attenuationShapeExtents {x=0}
    customAttenuationCurve {editorCurveData {keys [
       <at 0 1>
       {arriveTangent=0.028, interpMode=RCIM_Cubic, leaveTangent=0.028, tangentMode=RCTM_User, time=0.2209, value=1}
       {arriveTangent=-0.0561, interpMode=RCIM_Cubic, leaveTangent=-0.0561, tangentMode=RCTM_User, time=0.4, value=1}
       <at 1 0>
      ]}}
   }
  }
 }
 {
  id=Attenuation_3
  children [Modulator_2]
  properties {
   bOverrideAttenuation=true
   attenuationOverrides {
    bAttenuateWithLPF=true
    bEnableFocusInterpolation=true
    bEnableListenerFocus=true
    distanceAlgorithm=Custom
    falloffDistance=5000
    lPFFrequencyAtMax=750
    lPFFrequencyAtMin=12000
    lPFRadiusMax=4000
    lPFRadiusMin=1000
    nonFocusVolumeAttenuation=0.7
    nonSpatializedRadiusEnd=1000
    nonSpatializedRadiusStart=5000
    occlusionInterpolationTime=3
    occlusionLowPassFilterFrequency=1000
    occlusionTraceChannel=ECC_GameTraceChannel11
    stereoSpread=10000
    attenuationShapeExtents {x=0}
    customAttenuationCurve {editorCurveData {keys [
       <at 0 2>
       {arriveTangent=0.028, interpMode=RCIM_Cubic, leaveTangent=0.028, tangentMode=RCTM_User, time=0.2209, value=2}
       {arriveTangent=-3.1333, interpMode=RCIM_Cubic, leaveTangent=-3.1333, tangentMode=RCTM_User, time=0.3525, value=0.9707}
       {arriveTangent=-1.3783, interpMode=RCIM_Cubic, leaveTangent=-1.3783, tangentMode=RCTM_User, time=0.5659, value=0.4251}
       {arriveTangent=-2.6195, interpMode=RCIM_Cubic, leaveTangent=-2.6195, tangentMode=RCTM_User, time=1, value=0}
      ]}}
   }
  }
 }
 {
  id=Attenuation_6
  children [Modulator_1]
  properties {
   bOverrideAttenuation=true
   attenuationOverrides {
    bAttenuateWithLPF=true
    bEnableFocusInterpolation=true
    bEnableListenerFocus=true
    distanceAlgorithm=Logarithmic
    falloffDistance=7500
    focusAttackInterpSpeed=0.2
    lPFFrequencyAtMax=1000
    lPFRadiusMax=4000
    lPFRadiusMin=1000
    nonFocusVolumeAttenuation=0.8
    nonSpatializedRadiusStart=3000
    occlusionInterpolationTime=1
    occlusionLowPassFilterFrequency=1000
    occlusionTraceChannel=ECC_GameTraceChannel11
    stereoSpread=500
    attenuationShapeExtents {x=2000}
   }
  }
 }
 {
  id=Attenuation_7
  children [Modulator_3]
  properties {
   bOverrideAttenuation=true
   attenuationOverrides {
    bAttenuateWithLPF=true
    bEnableFocusInterpolation=true
    bEnableListenerFocus=true
    distanceAlgorithm=Logarithmic
    falloffDistance=7500
    focusAttackInterpSpeed=0.2
    lPFFrequencyAtMax=1000
    lPFRadiusMax=4000
    lPFRadiusMin=1000
    nonFocusVolumeAttenuation=0.8
    nonSpatializedRadiusStart=3000
    occlusionInterpolationTime=1
    occlusionLowPassFilterFrequency=1000
    occlusionTraceChannel=ECC_GameTraceChannel11
    stereoSpread=500
    attenuationShapeExtents {x=2000}
   }
  }
 }
 {id=Delay_0, children [Attenuation_3], properties {delayMax=0.5, delayMin=0.2}}
 {id=Mixer_0, children [Attenuation_2, Delay_0], properties {inputVolume [0.5,1.5]}}
 {id=Mixer_2, children [null,SoundClass_0], properties {inputVolume [1,1]}}
 {id=Modulator_1, children [WavePlayer_5], properties {pitchMax=1.1, pitchMin=0.9}}
 {id=Modulator_2, children [WavePlayer_2], properties {pitchMax=1.3, pitchMin=1.2}}
 {id=Modulator_3, children [WavePlayer_9], properties {pitchMax=1.1, pitchMin=0.9}}
 {id=Modulator_4, children [WavePlayer_3], properties {pitchMax=1.3, pitchMin=1.2}}
 {id=QualityLevel_0, children [Mixer_2, Mixer_2, Attenuation_6, Attenuation_7]}
 {id=SoundClass_0, children [Mixer_0], properties {soundClassOverride=/Game/Sounds/FortSoundClassesAndModes/Fort_Ambient_Water}}
 {id=WavePlayer_2, properties {bLooping=true, soundWaveAssetPtr=/Game/Athena/Sounds/Water/Ambient/Amb_Water_Waterfall_Large_Bot_Loop_01}}
 {
  id=WavePlayer_3
  properties {bLooping=true, soundWaveAssetPtr=/Game/Sounds/Artemis/POI/RealityFalls/Spots/Waterfalls/Spot_RealityFalls_Waterfall_Roar_Medium_Distant_01}
 }
 {
  id=WavePlayer_5
  properties {bLooping=true, soundWaveAssetPtr=/Game/Sounds/Artemis/POI/RealityFalls/Spots/Waterfalls/Spot_RealityFalls_Waterfall_Roar_Close}
 }
 {
  id=WavePlayer_9
  properties {bLooping=true, soundWaveAssetPtr=&Spot_RealityFalls_Waterfall_Roar_Close}
 }
]
properties {
 bOverrideConcurrency=true
 maxDistance=9500
 soundClassObject=/Game/Sounds/FortSoundClassesAndModes/Fort_Water_Ambient
 volumeMultiplier=0.251
 attenuationOverrides {
  bAttenuateWithLPF=true
  bEnableListenerFocus=true
  falloffDistance=6000
  lPFFrequencyAtMax=500
  nonSpatializedRadiusStart=700
  stereoSpread=1000
  attenuationShapeExtents {x=925}
 }
 concurrencyOverrides {maxCount=2, voiceStealReleaseTime=3}
}
