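# Underworld biome ambient one-shot cue that plays distant voice howls as "whispers" spots, with distant-creature attenuation and one instance at a time.
# It randomly picks one of two pitch-enveloped, randomly modulated branches, each choosing among the same three howl waves.
# One branch adds volume swells on two waves; the other adds random volume variation.
# Only the first two quality levels are wired, so the others have no sound.
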SoundCue /Game/Sounds/Helios/Biomes/Underworld/Spots/Spot_Underworld_Whispers_Cue
root=QualityLevel_0
nodes [
 {
  id=Enveloper_0
  children [Modulator_0]
  properties {pitchCurve {editorCurveData {keys [{time=0}, {time=1}, <at 7.5 0.8>]}}, volumeCurve {editorCurveData {keys [<at 0 1>]}}}
 }
 {
  id=Enveloper_1
  children [WavePlayer_10]
  properties {pitchCurve {editorCurveData {keys [{time=0}]}}, volumeCurve {editorCurveData {keys [<at 0 0.5>,<at 2 0.5>,<at 3 2>]}}}
 }
 {
  id=Enveloper_2
  children [WavePlayer_11]
  properties {pitchCurve {editorCurveData {keys [{time=0}]}}, volumeCurve {editorCurveData {keys [<at 0 0.5>,<at 2 0.5>,<at 3 1.5>]}}}
 }
 {
  id=Enveloper_5
  children [Modulator_1]
  properties {pitchCurve {editorCurveData {keys [{time=0}, {time=1}, <at 7.5 0.95>]}}, volumeCurve {editorCurveData {keys [<at 0 1>]}}}
 }
 {id=Mixer_0, children [WavePlayer_12], properties {inputVolume [0.75]}}
 {id=Mixer_1, children [WavePlayer_2], properties {inputVolume [0.75]}}
 {id=Modulator_0, children [Random_2], properties {pitchMax=1.1, pitchMin=0.7}}
 {id=Modulator_1, children [Random_0], properties {pitchMax=1.1, pitchMin=0.7, volumeMax=0.9, volumeMin=0.8}}
 {id=QualityLevel_0, children [Random_1,Random_1,null,null]}
 {id=Random_0, children [WavePlayer_0,WavePlayer_1,Mixer_1], properties {weights [1,1,1]}}
 {id=Random_1, children [Enveloper_0,Enveloper_5], properties {weights [1,1]}}
 {id=Random_2, children [Enveloper_1,Enveloper_2,Mixer_0], properties {weights [1,1,1]}}
 {id=WavePlayer_0, properties {soundWaveAssetPtr=/Game/Sounds/Creative/Ambiences/AbstractVoices/Voice_HowlDistantB001_Stereo}}
 {id=WavePlayer_1, properties {soundWaveAssetPtr=/Game/Sounds/Creative/Ambiences/AbstractVoices/Voice_HowlDistantB004_Stereo}}
 {id=WavePlayer_10, properties {soundWaveAssetPtr=&Voice_HowlDistantB001_Stereo}}
 {id=WavePlayer_11, properties {soundWaveAssetPtr=&Voice_HowlDistantB004_Stereo}}
 {id=WavePlayer_12, properties {soundWaveAssetPtr=/Game/Sounds/Creative/Ambiences/AbstractVoices/Voice_HowlDistantB006_Stereo}}
 {id=WavePlayer_2, properties {soundWaveAssetPtr=&Voice_HowlDistantB006_Stereo}}
]
properties {
 attenuationSettings=/Game/Sounds/Attenuation/Fort_TileAmbient_OneShot_Distant_Creature_Att
 maxDistance=8500
 soundClassObject=/Game/Sounds/FortSoundClassesAndModes/Fort_Ambient_OneShots_Indoors
 volumeMultiplier=0.401
 concurrencyOverrides {maxCount=1, resolutionRule=PreventNew}
}
