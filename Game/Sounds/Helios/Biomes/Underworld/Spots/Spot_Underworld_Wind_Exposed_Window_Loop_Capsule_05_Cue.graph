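# Looping ambient cue for howling window wind at Underworld biome spots, classed as indoor ambience.
# One looping wave feeds two nested, endlessly looping Envelopers with different periods, a pitch curve inside and a volume swell outside, so the variation drifts.
# The Quality Level root routes that chain to its first two slots and leaves the other two empty.
# Attenuation references a shared interior window wind asset, and a concurrency override caps simultaneous instances.
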SoundCue /Game/Sounds/Helios/Biomes/Underworld/Spots/Spot_Underworld_Wind_Exposed_Window_Loop_Capsule_05_Cue
root=QualityLevel_1
nodes [
 {
  id=Enveloper_1
  children [WavePlayer_3]
  properties {
   bLoop=true
   bLoopIndefinitely=true
   loopEnd=30
   pitchCurve {editorCurveData {keys [{interpMode=RCIM_Cubic, time=0}, <at 15.9468 0.9181 RCIM_Cubic>, {interpMode=RCIM_Cubic, time=30}]}}
   volumeCurve {editorCurveData {keys [<at 0 1>]}}
  }
 }
 {
  id=Enveloper_2
  children [Enveloper_1]
  properties {
   bLoop=true
   bLoopIndefinitely=true
   loopEnd=20
   pitchCurve {editorCurveData {keys [{time=0}]}}
   volumeCurve {editorCurveData {keys [
      {arriveTangent=-0.01, interpMode=RCIM_Cubic, time=0, value=1}
      <at 10 1.2 RCIM_Cubic>
      {interpMode=RCIM_Cubic, leaveTangent=0.01, time=20, value=1}
     ]}}
  }
 }
 {id=QualityLevel_1, children [Enveloper_2,Enveloper_2,null,null]}
 {
  id=WavePlayer_3
  properties {bLooping=true, soundWaveAssetPtr=&/Ambience_Wind_Window_Howling_Loop_02}
 }
]
properties {
 attenuationSettings=/Game/Sounds/Artemis/POI/Hollow_Base/Spots/Spot_Interior_Window_Wind_Att
 bOverrideConcurrency=true
 maxDistance=1253.7231
 soundClassObject=&Fort_Ambient_Indoors_Athena
 volumeMultiplier=0.501
 attenuationOverrides {
  attenuationShape=Capsule
  bAttenuateWithLPF=true
  bEnableListenerFocus=true
  bEnableLogFrequencyScaling=true
  bEnableSubmixSends=true
  binauralRadius=400
  distanceAlgorithm=Logarithmic
  falloffDistance=625
  lPFFrequencyAtMax=500
  lPFRadiusMax=2000
  lPFRadiusMin=400
  nonFocusVolumeAttenuation=0.75
  spatializationAlgorithm=SPATIALIZATION_HRTF
  attenuationShapeExtents=<1250 180>
  submixSendSettings [{
    bEnableHPFCutoff=false
    bEnableLPFCutoff=false
    disableManualSendClamp=false
    hPFCutoff=20
    lPFCutoff=20000
    maxSendDistance=7500
    maxSendLevel=0.25
    minSendDistance=400
    minSendLevel=0
    sendLevel=0.2
    sendLevelControlMethod=Linear
    soundSubmix=&Fortnite_Reverb_Distant
    customSendLevelCurve {externalCurve=/Game/Sounds/SubMixes/Reverb/Fortnite_Reverb_Distant_Send}
   }]
 }
 concurrencyOverrides {maxCount=2}
}
