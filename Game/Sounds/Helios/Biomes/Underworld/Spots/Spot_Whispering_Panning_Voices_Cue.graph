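# Sound cue for Underworld biome ambient one-shot spots of whispering, panning voices, using wider tile-ambient attenuation and an exterior ambient bus send.
# A quality-level node sends its first two slots to a random pool of 22 waves and its last two to a six-wave pool excluded from branch culling.
# Each pool passes through slight pitch and volume modulation.
# Concurrency allows one instance and blocks new plays while it sounds.
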SoundCue /Game/Sounds/Helios/Biomes/Underworld/Spots/Spot_Whispering_Panning_Voices_Cue
root=QualityLevel_0
nodes [
 {id=Modulator_2, children [Random_0], properties {pitchMax=1.2, pitchMin=0.8, volumeMax=1.1, volumeMin=0.9}}
 {id=Modulator_3, children [Random_1], properties {pitchMax=1.2, pitchMin=0.8, volumeMax=1.1, volumeMin=0.9}}
 {id=QualityLevel_0, children [Modulator_2,Modulator_2,Modulator_3,Modulator_3]}
 {
  id=Random_0
  children [
   WavePlayer_0
   WavePlayer_1
   WavePlayer_2
   WavePlayer_3
   WavePlayer_4
   WavePlayer_5
   WavePlayer_6
   WavePlayer_7
   WavePlayer_8
   WavePlayer_9
   WavePlayer_10
   WavePlayer_11
   WavePlayer_12
   WavePlayer_13
   WavePlayer_14
   WavePlayer_15
   WavePlayer_16
   WavePlayer_17
   WavePlayer_18
   WavePlayer_19
   WavePlayer_20
   WavePlayer_21
  ]
  properties {weights [1,1,1,1,1,1,1,1,1,1,1,1,1,1,1,1,1,1,1,1,1,1]}
 }
 {
  id=Random_1
  children [WavePlayer_22, WavePlayer_23, WavePlayer_24, WavePlayer_25, WavePlayer_66, WavePlayer_74]
  properties {bShouldExcludeFromBranchCulling=true, weights [1,1,1,1,1,1]}
 }
 {id=WavePlayer_0, properties {soundWaveAssetPtr=&/Spot_Whispering_Panning_Voices_01}}
 {id=WavePlayer_1, properties {soundWaveAssetPtr=&/Spot_Whispering_Panning_Voices_02}}
 {id=WavePlayer_10, properties {soundWaveAssetPtr=&/Spot_Whispering_Panning_Voices_11}}
 {id=WavePlayer_11, properties {soundWaveAssetPtr=&/Spot_Whispering_Panning_Voices_12}}
 {id=WavePlayer_12, properties {soundWaveAssetPtr=&/Spot_Whispering_Panning_Voices_13}}
 {id=WavePlayer_13, properties {soundWaveAssetPtr=&/Spot_Whispering_Panning_Voices_14}}
 {id=WavePlayer_14, properties {soundWaveAssetPtr=&/Spot_Whispering_Panning_Voices_15}}
 {id=WavePlayer_15, properties {soundWaveAssetPtr=&/Spot_Whispering_Panning_Voices_16}}
 {id=WavePlayer_16, properties {soundWaveAssetPtr=&/Spot_Whispering_Panning_Voices_17}}
 {id=WavePlayer_17, properties {soundWaveAssetPtr=&/Spot_Whispering_Panning_Voices_18}}
 {id=WavePlayer_18, properties {soundWaveAssetPtr=&/Spot_Whispering_Panning_Voices_19}}
 {id=WavePlayer_19, properties {soundWaveAssetPtr=&/Spot_Whispering_Panning_Voices_20}}
 {id=WavePlayer_2, properties {soundWaveAssetPtr=&/Spot_Whispering_Panning_Voices_03}}
 {id=WavePlayer_20, properties {soundWaveAssetPtr=&/Spot_Whispering_Panning_Voices_21}}
 {id=WavePlayer_21, properties {soundWaveAssetPtr=&/Spot_Whispering_Panning_Voices_22}}
 {id=WavePlayer_22, properties {soundWaveAssetPtr=&Spot_Whispering_Panning_Voices_22}}
 {id=WavePlayer_23, properties {soundWaveAssetPtr=&Spot_Whispering_Panning_Voices_20}}
 {id=WavePlayer_24, properties {soundWaveAssetPtr=&Spot_Whispering_Panning_Voices_21}}
 {id=WavePlayer_25, properties {soundWaveAssetPtr=&Spot_Whispering_Panning_Voices_01}}
 {id=WavePlayer_3, properties {soundWaveAssetPtr=&/Spot_Whispering_Panning_Voices_04}}
 {id=WavePlayer_4, properties {soundWaveAssetPtr=&/Spot_Whispering_Panning_Voices_05}}
 {id=WavePlayer_5, properties {soundWaveAssetPtr=&/Spot_Whispering_Panning_Voices_06}}
 {id=WavePlayer_6, properties {soundWaveAssetPtr=&/Spot_Whispering_Panning_Voices_07}}
 {id=WavePlayer_66, properties {soundWaveAssetPtr=&Spot_Whispering_Panning_Voices_02}}
 {id=WavePlayer_7, properties {soundWaveAssetPtr=&/Spot_Whispering_Panning_Voices_08}}
 {id=WavePlayer_74, properties {soundWaveAssetPtr=&Spot_Whispering_Panning_Voices_03}}
 {id=WavePlayer_8, properties {soundWaveAssetPtr=&/Spot_Whispering_Panning_Voices_09}}
 {id=WavePlayer_9, properties {soundWaveAssetPtr=&/Spot_Whispering_Panning_Voices_10}}
]
properties {
 attenuationSettings=/Game/Sounds/Attenuation/Fort_TileAmbient_OneShot_Wider_Att
 bOverrideConcurrency=true
 maxDistance=8000
 soundClassObject=&Fort_Ambient_OneShots
 volumeMultiplier=0.301
 busSends [{audioBus=null, minSendLevel=1, sendLevel=1, soundSourceBus=&AmbientAudio_Exterior_SourceBus}]
 concurrencyOverrides {maxCount=1, resolutionRule=PreventNew, voiceStealReleaseTime=2}
}
