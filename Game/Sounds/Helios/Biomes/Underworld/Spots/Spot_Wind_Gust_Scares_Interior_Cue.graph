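# Sound cue for scary interior wind-gust one-shot spot ambience in the Helios Underworld biome.
# An equal-weight random picks chittering voices, ghost whispers or hollow rasps (slightly lowered via a mixer), then a random variant within that set, under pitch modulation.
# Only the first two quality-level slots are connected; the other two are empty.
# Concurrency allows one instance and blocks new ones, with a distance-scaled send to distant reverb.
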SoundCue /Game/Sounds/Helios/Biomes/Underworld/Spots/Spot_Wind_Gust_Scares_Interior_Cue
root=Modulator_0
nodes [
 {id=Mixer_0, children [Random_2], properties {inputVolume [0.9]}}
 {id=Modulator_0, children [QualityLevel_0], properties {pitchMin=0.8}}
 {id=QualityLevel_0, children [Random_6,Random_6,null,null]}
 {
  id=Random_2
  children [
   WavePlayer_1
   WavePlayer_12
   WavePlayer_13
   WavePlayer_14
   WavePlayer_15
   WavePlayer_18
   WavePlayer_19
   WavePlayer_20
   WavePlayer_22
   WavePlayer_23
   WavePlayer_24
  ]
  properties {weights [1,1,1,1,1,1,1,1,1,1,1]}
 }
 {
  id=Random_4
  children [
   WavePlayer_26
   WavePlayer_27
   WavePlayer_28
   WavePlayer_29
   WavePlayer_30
   WavePlayer_31
   WavePlayer_32
   WavePlayer_33
   WavePlayer_34
   WavePlayer_35
   WavePlayer_36
  ]
  properties {weights [1,1,1,1,1,1,1,1,1,1,1]}
 }
 {
  id=Random_5
  children [WavePlayer_37, WavePlayer_38, WavePlayer_39, WavePlayer_40, WavePlayer_41, WavePlayer_42, WavePlayer_43, WavePlayer_44, WavePlayer_45]
  properties {weights [1,1,1,1,1,1,1,1,1]}
 }
 {id=Random_6, children [Random_5,Random_4,Mixer_0], properties {weights [1,1,1]}}
 {id=WavePlayer_1, properties {soundWaveAssetPtr=&/Spot_Wind_Gust_Scary_Hollow_Rasp_01}}
 {id=WavePlayer_12, properties {soundWaveAssetPtr=&/Spot_Wind_Gust_Scary_Hollow_Rasp_02}}
 {id=WavePlayer_13, properties {soundWaveAssetPtr=&/Spot_Wind_Gust_Scary_Hollow_Rasp_03}}
 {id=WavePlayer_14, properties {soundWaveAssetPtr=&/Spot_Wind_Gust_Scary_Hollow_Rasp_04}}
 {id=WavePlayer_15, properties {soundWaveAssetPtr=&/Spot_Wind_Gust_Scary_Hollow_Rasp_05}}
 {id=WavePlayer_18, properties {soundWaveAssetPtr=&/Spot_Wind_Gust_Scary_Hollow_Rasp_06}}
 {id=WavePlayer_19, properties {soundWaveAssetPtr=&/Spot_Wind_Gust_Scary_Hollow_Rasp_07}}
 {id=WavePlayer_20, properties {soundWaveAssetPtr=&/Spot_Wind_Gust_Scary_Hollow_Rasp_08}}
 {id=WavePlayer_22, properties {soundWaveAssetPtr=&/Spot_Wind_Gust_Scary_Hollow_Rasp_10}}
 {id=WavePlayer_23, properties {soundWaveAssetPtr=&/Spot_Wind_Gust_Scary_Hollow_Rasp_11}}
 {id=WavePlayer_24, properties {soundWaveAssetPtr=&/Spot_Wind_Gust_Scary_Hollow_Rasp_12}}
 {id=WavePlayer_26, properties {soundWaveAssetPtr=&/Spot_Wind_Gust_Scary_Ghost_Whispers_01}}
 {id=WavePlayer_27, properties {soundWaveAssetPtr=&/Spot_Wind_Gust_Scary_Ghost_Whispers_02}}
 {id=WavePlayer_28, properties {soundWaveAssetPtr=&/Spot_Wind_Gust_Scary_Ghost_Whispers_03}}
 {id=WavePlayer_29, properties {soundWaveAssetPtr=&/Spot_Wind_Gust_Scary_Ghost_Whispers_04}}
 {id=WavePlayer_30, properties {soundWaveAssetPtr=&/Spot_Wind_Gust_Scary_Ghost_Whispers_05}}
 {id=WavePlayer_31, properties {soundWaveAssetPtr=&/Spot_Wind_Gust_Scary_Ghost_Whispers_06}}
 {id=WavePlayer_32, properties {soundWaveAssetPtr=&/Spot_Wind_Gust_Scary_Ghost_Whispers_07}}
 {id=WavePlayer_33, properties {soundWaveAssetPtr=&/Spot_Wind_Gust_Scary_Ghost_Whispers_08}}
 {id=WavePlayer_34, properties {soundWaveAssetPtr=&/Spot_Wind_Gust_Scary_Ghost_Whispers_09}}
 {id=WavePlayer_35, properties {soundWaveAssetPtr=&/Spot_Wind_Gust_Scary_Ghost_Whispers_10}}
 {id=WavePlayer_36, properties {soundWaveAssetPtr=&/Spot_Wind_Gust_Scary_Ghost_Whispers_11}}
 {id=WavePlayer_37, properties {soundWaveAssetPtr=&/Spot_Wind_Gust_Scary_Chittering_Voice_01}}
 {id=WavePlayer_38, properties {soundWaveAssetPtr=&/Spot_Wind_Gust_Scary_Chittering_Voice_02}}
 {id=WavePlayer_39, properties {soundWaveAssetPtr=&/Spot_Wind_Gust_Scary_Chittering_Voice_03}}
 {id=WavePlayer_40, properties {soundWaveAssetPtr=&/Spot_Wind_Gust_Scary_Chittering_Voice_04}}
 {id=WavePlayer_41, properties {soundWaveAssetPtr=&/Spot_Wind_Gust_Scary_Chittering_Voice_05}}
 {id=WavePlayer_42, properties {soundWaveAssetPtr=&/Spot_Wind_Gust_Scary_Chittering_Voice_06}}
 {id=WavePlayer_43, properties {soundWaveAssetPtr=&/Spot_Wind_Gust_Scary_Chittering_Voice_07}}
 {id=WavePlayer_44, properties {soundWaveAssetPtr=&/Spot_Wind_Gust_Scary_Chittering_Voice_08}}
 {id=WavePlayer_45, properties {soundWaveAssetPtr=&/Spot_Wind_Gust_Scary_Chittering_Voice_09}}
]
properties {
 attenuationSettings=/Game/Sounds/Attenuation/Fort_TileAmbient_OneShot_Distant_Creature_Att
 bEnableBusSends=false
 bOverrideConcurrency=true
 maxDistance=8500
 soundClassObject=&Fort_Ambient_OneShots
 volumeMultiplier=0.401
 attenuationOverrides {bAttenuate=false}
 concurrencyOverrides {maxCount=1, resolutionRule=PreventNew, voiceStealReleaseTime=2}
 soundSubmixSends [{
   maxSendDistance=8000
   minSendLevel=0.25
   soundSubmix=&Fortnite_Reverb_Distant
   customSendLevelCurve {editorCurveData {keys [<at 0 0.25>,<at 8000 1>]}}
  }]
}
