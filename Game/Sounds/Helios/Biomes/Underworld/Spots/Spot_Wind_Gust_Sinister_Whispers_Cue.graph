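# Ambient one-shot Sound Cue for Underworld biome spots, playing sinister wind gusts with whispers under outdoor ambient sound class and attenuation.
# A quality-level node sends its first two levels to a mixer layering random whisper-tone gusts with randomly delayed panning whisper voices.
# The last two levels use a lighter branch that picks from six tone clips only.
# Modulators vary volume on every branch and pitch on the tone layers.
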SoundCue /Game/Sounds/Helios/Biomes/Underworld/Spots/Spot_Wind_Gust_Sinister_Whispers_Cue
root=QualityLevel_0
nodes [
 {id=Delay_0, children [Modulator_1], properties {delayMax=2, delayMin=0.5}}
 {id=Mixer_0, children [Modulator_2,Delay_0], properties {inputVolume [1.1,1]}}
 {id=Modulator_0, children [Random_6], properties {pitchMax=1.1, pitchMin=0.7, volumeMax=1.1, volumeMin=0.5}}
 {id=Modulator_1, children [Random_0], properties {volumeMax=1.1, volumeMin=0.9}}
 {id=Modulator_2, children [Random_4], properties {pitchMax=1.2, pitchMin=0.8, volumeMax=1.1, volumeMin=0.9}}
 {id=QualityLevel_0, children [Mixer_0,Mixer_0,Modulator_0,Modulator_0]}
 {
  id=Random_0
  children [
   WavePlayer_0
   WavePlayer_1
   WavePlayer_2
   WavePlayer_3
   WavePlayer_4
   WavePlayer_5
   WavePlayer_6
   WavePlayer_7
   WavePlayer_8
   WavePlayer_9
   WavePlayer_10
   WavePlayer_11
   WavePlayer_12
   WavePlayer_13
   WavePlayer_14
   WavePlayer_15
   WavePlayer_16
   WavePlayer_17
   WavePlayer_18
   WavePlayer_19
   WavePlayer_20
   WavePlayer_21
  ]
  properties {weights [1,1,1,1,1,1,1,1,1,1,1,1,1,1,1,1,1,1,1,1,1,1]}
 }
 {
  id=Random_4
  children [
   WavePlayer_26
   WavePlayer_27
   WavePlayer_28
   WavePlayer_29
   WavePlayer_30
   WavePlayer_31
   WavePlayer_32
   WavePlayer_33
   WavePlayer_34
   WavePlayer_35
   WavePlayer_36
   WavePlayer_37
   WavePlayer_38
   WavePlayer_39
   WavePlayer_40
   WavePlayer_41
   WavePlayer_42
   WavePlayer_43
   WavePlayer_44
   WavePlayer_45
   WavePlayer_46
  ]
  properties {weights [1,1,1,1,1,1,1,1,1,1,1,1,1,1,1,1,1,1,1,1,1]}
 }
 {
  id=Random_6
  children [WavePlayer_71, WavePlayer_72, WavePlayer_73, WavePlayer_68, WavePlayer_69, WavePlayer_70]
  properties {bShouldExcludeFromBranchCulling=true, weights [1,1,1,1,1,1]}
 }
 {id=WavePlayer_0, properties {soundWaveAssetPtr=&/Spot_Whispering_Panning_Voices_01}}
 {id=WavePlayer_1, properties {soundWaveAssetPtr=&/Spot_Whispering_Panning_Voices_02}}
 {id=WavePlayer_10, properties {soundWaveAssetPtr=&/Spot_Whispering_Panning_Voices_11}}
 {id=WavePlayer_11, properties {soundWaveAssetPtr=&/Spot_Whispering_Panning_Voices_12}}
 {id=WavePlayer_12, properties {soundWaveAssetPtr=&/Spot_Whispering_Panning_Voices_13}}
 {id=WavePlayer_13, properties {soundWaveAssetPtr=&/Spot_Whispering_Panning_Voices_14}}
 {id=WavePlayer_14, properties {soundWaveAssetPtr=&/Spot_Whispering_Panning_Voices_15}}
 {id=WavePlayer_15, properties {soundWaveAssetPtr=&/Spot_Whispering_Panning_Voices_16}}
 {id=WavePlayer_16, properties {soundWaveAssetPtr=&/Spot_Whispering_Panning_Voices_17}}
 {id=WavePlayer_17, properties {soundWaveAssetPtr=&/Spot_Whispering_Panning_Voices_18}}
 {id=WavePlayer_18, properties {soundWaveAssetPtr=&/Spot_Whispering_Panning_Voices_19}}
 {id=WavePlayer_19, properties {soundWaveAssetPtr=&/Spot_Whispering_Panning_Voices_20}}
 {id=WavePlayer_2, properties {soundWaveAssetPtr=&/Spot_Whispering_Panning_Voices_03}}
 {id=WavePlayer_20, properties {soundWaveAssetPtr=&/Spot_Whispering_Panning_Voices_21}}
 {id=WavePlayer_21, properties {soundWaveAssetPtr=&/Spot_Whispering_Panning_Voices_22}}
 {id=WavePlayer_26, properties {soundWaveAssetPtr=&/Spot_Wind_Gust_Sinister_Whisper_Tone_01}}
 {id=WavePlayer_27, properties {soundWaveAssetPtr=&/Spot_Wind_Gust_Sinister_Whisper_Tone_02}}
 {id=WavePlayer_28, properties {soundWaveAssetPtr=&/Spot_Wind_Gust_Sinister_Whisper_Tone_03}}
 {id=WavePlayer_29, properties {soundWaveAssetPtr=&/Spot_Wind_Gust_Sinister_Whisper_Tone_04}}
 {id=WavePlayer_3, properties {soundWaveAssetPtr=&/Spot_Whispering_Panning_Voices_04}}
 {id=WavePlayer_30, properties {soundWaveAssetPtr=&/Spot_Wind_Gust_Sinister_Whisper_Tone_05}}
 {id=WavePlayer_31, properties {soundWaveAssetPtr=&/Spot_Wind_Gust_Sinister_Whisper_Tone_06}}
 {id=WavePlayer_32, properties {soundWaveAssetPtr=&/Spot_Wind_Gust_Sinister_Whisper_Tone_07}}
 {id=WavePlayer_33, properties {soundWaveAssetPtr=&/Spot_Wind_Gust_Sinister_Whisper_Tone_08}}
 {id=WavePlayer_34, properties {soundWaveAssetPtr=&/Spot_Wind_Gust_Sinister_Whisper_Tone_09}}
 {id=WavePlayer_35, properties {soundWaveAssetPtr=&/Spot_Wind_Gust_Sinister_Whisper_Tone_10}}
 {id=WavePlayer_36, properties {soundWaveAssetPtr=&/Spot_Wind_Gust_Sinister_Whisper_Tone_11}}
 {id=WavePlayer_37, properties {soundWaveAssetPtr=&/Spot_Wind_Gust_Sinister_Whisper_Tone_12}}
 {id=WavePlayer_38, properties {soundWaveAssetPtr=&/Spot_Wind_Gust_Sinister_Whisper_Tone_13}}
 {id=WavePlayer_39, properties {soundWaveAssetPtr=&/Spot_Wind_Gust_Sinister_Whisper_Tone_14}}
 {id=WavePlayer_4, properties {soundWaveAssetPtr=&/Spot_Whispering_Panning_Voices_05}}
 {id=WavePlayer_40, properties {soundWaveAssetPtr=&/Spot_Wind_Gust_Sinister_Whisper_Tone_15}}
 {id=WavePlayer_41, properties {soundWaveAssetPtr=&/Spot_Wind_Gust_Sinister_Whisper_Tone_16}}
 {id=WavePlayer_42, properties {soundWaveAssetPtr=&/Spot_Wind_Gust_Sinister_Whisper_Tone_17}}
 {id=WavePlayer_43, properties {soundWaveAssetPtr=&/Spot_Wind_Gust_Sinister_Whisper_Tone_18}}
 {id=WavePlayer_44, properties {soundWaveAssetPtr=&/Spot_Wind_Gust_Sinister_Whisper_Tone_19}}
 {id=WavePlayer_45, properties {soundWaveAssetPtr=&/Spot_Wind_Gust_Sinister_Whisper_Tone_20}}
 {id=WavePlayer_46, properties {soundWaveAssetPtr=&/Spot_Wind_Gust_Sinister_Whisper_Tone_21}}
 {id=WavePlayer_5, properties {soundWaveAssetPtr=&/Spot_Whispering_Panning_Voices_06}}
 {id=WavePlayer_6, properties {soundWaveAssetPtr=&/Spot_Whispering_Panning_Voices_07}}
 {id=WavePlayer_68, properties {soundWaveAssetPtr=&/Spot_Wind_Gust_Sinister_Whisper_Tone_40}}
 {id=WavePlayer_69, properties {soundWaveAssetPtr=&/Spot_Wind_Gust_Sinister_Whisper_Tone_41}}
 {id=WavePlayer_7, properties {soundWaveAssetPtr=&/Spot_Whispering_Panning_Voices_08}}
 {id=WavePlayer_70, properties {soundWaveAssetPtr=&/Spot_Wind_Gust_Sinister_Whisper_Tone_42}}
 {id=WavePlayer_71, properties {soundWaveAssetPtr=&Spot_Wind_Gust_Sinister_Whisper_Tone_19}}
 {id=WavePlayer_72, properties {soundWaveAssetPtr=&Spot_Wind_Gust_Sinister_Whisper_Tone_20}}
 {id=WavePlayer_73, properties {soundWaveAssetPtr=&Spot_Wind_Gust_Sinister_Whisper_Tone_21}}
 {id=WavePlayer_8, properties {soundWaveAssetPtr=&/Spot_Whispering_Panning_Voices_09}}
 {id=WavePlayer_9, properties {soundWaveAssetPtr=&/Spot_Whispering_Panning_Voices_10}}
]
properties {
 attenuationSettings=/Game/Sounds/Attenuation/Fort_TileAmbient_OneShot_Wider_Att
 maxDistance=8000
 soundClassObject=&Fort_Ambient_Outdoors_Athena
 volumeMultiplier=0.351
 preEffectBusSends [{}]
}
