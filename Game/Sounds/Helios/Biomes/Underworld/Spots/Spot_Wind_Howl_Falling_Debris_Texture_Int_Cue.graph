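# Indoor ambient one-shot cue for Underworld biome spots, playing wind howl and falling debris textures reused from Jungle Temple recordings.
# A quality switch sends its first two levels to a random pick of eight waves and its last two to a pick of the first four.
# Each branch has its own pitch and volume modulation.
# Playback uses wide tile-ambient attenuation, the indoor ambient sound class and a concurrency cap of two.
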SoundCue /Game/Sounds/Helios/Biomes/Underworld/Spots/Spot_Wind_Howl_Falling_Debris_Texture_Int_Cue
root=QualityLevel_0
nodes [
 {id=Modulator_1, children [Random_2], properties {pitchMax=1.1, pitchMin=0.9, volumeMax=1.4, volumeMin=0.7}}
 {id=Modulator_2, children [Random_3], properties {pitchMax=1.1, volumeMax=1.1, volumeMin=0.1}}
 {id=QualityLevel_0, children [Modulator_1,Modulator_1,Modulator_2,Modulator_2]}
 {
  id=Random_2
  children [WavePlayer_12, WavePlayer_13, WavePlayer_14, WavePlayer_15, WavePlayer_16, WavePlayer_17, WavePlayer_18, WavePlayer_19]
  properties {weights [1,1,1,1,1,1,1,1]}
 }
 {id=Random_3, children [WavePlayer_22, WavePlayer_23, WavePlayer_24, WavePlayer_25], properties {weights [1,1,1,1]}}
 {id=WavePlayer_12, properties {soundWaveAssetPtr=/Game/Sounds/Artemis/POI/Jungle_Temple/Spots/AMB_Canary_Jungle_Temple_Int_SS_Wind_Debris_01}}
 {id=WavePlayer_13, properties {soundWaveAssetPtr=/Game/Sounds/Artemis/POI/Jungle_Temple/Spots/AMB_Canary_Jungle_Temple_Int_SS_Wind_Debris_02}}
 {id=WavePlayer_14, properties {soundWaveAssetPtr=/Game/Sounds/Artemis/POI/Jungle_Temple/Spots/AMB_Canary_Jungle_Temple_Int_SS_Wind_Debris_03}}
 {id=WavePlayer_15, properties {soundWaveAssetPtr=/Game/Sounds/Artemis/POI/Jungle_Temple/Spots/AMB_Canary_Jungle_Temple_Int_SS_Wind_Debris_04}}
 {id=WavePlayer_16, properties {soundWaveAssetPtr=/Game/Sounds/Artemis/POI/Jungle_Temple/Spots/AMB_Canary_Jungle_Temple_Int_SS_Wind_Debris_05}}
 {id=WavePlayer_17, properties {soundWaveAssetPtr=/Game/Sounds/Artemis/POI/Jungle_Temple/Spots/AMB_Canary_Jungle_Temple_Int_SS_Wind_Debris_06}}
 {id=WavePlayer_18, properties {soundWaveAssetPtr=/Game/Sounds/Artemis/POI/Jungle_Temple/Spots/AMB_Canary_Jungle_Temple_Int_SS_Wind_Debris_07}}
 {id=WavePlayer_19, properties {soundWaveAssetPtr=/Game/Sounds/Artemis/POI/Jungle_Temple/Spots/AMB_Canary_Jungle_Temple_Int_SS_Wind_Debris_08}}
 {id=WavePlayer_22, properties {soundWaveAssetPtr=&AMB_Canary_Jungle_Temple_Int_SS_Wind_Debris_01}}
 {id=WavePlayer_23, properties {soundWaveAssetPtr=&AMB_Canary_Jungle_Temple_Int_SS_Wind_Debris_02}}
 {id=WavePlayer_24, properties {soundWaveAssetPtr=&AMB_Canary_Jungle_Temple_Int_SS_Wind_Debris_03}}
 {id=WavePlayer_25, properties {soundWaveAssetPtr=&AMB_Canary_Jungle_Temple_Int_SS_Wind_Debris_04}}
]
properties {
 attenuationSettings=/Game/Sounds/Attenuation/Fort_TileAmbient_OneShot_Wider_Att
 bEnableBusSends=false
 bOverrideConcurrency=true
 maxDistance=8000
 soundClassObject=&Fort_Ambient_Indoors_Athena
 volumeMultiplier=1.002
 concurrencyOverrides {maxCount=2}
 preEffectBusSends [{}]
}
